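# UI material that draws the icon background for Reclamation activity UI: a glowing, outlined box with a tinted icon on top.
# A box SDF, blended toward a 9-slice box by "Defense bg", yields outline, fill and glow masks.
# A grayscale icon, read as a distance field, is thresholded twice for shape and shadow.
# "CheckMark Switch" blends to a checkmark and recolors, while static switch "RemoveIcon" omits the icon layer.
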Material /TidalCrane_BaseActivities_UI/UI/Materials/M_UI_Reclamation_IconBg
graph {
 properties {bAutomaticallySetUsageInEditor=false, blendMode=BLEND_AlphaComposite, materialDomain=MD_UI, usedShadingModels=Default Lit}
 nodes [
  n1 = NamedRerouteDeclaration(None=MaterialFunctionCall(UV=MaterialFunctionCall() {materialFunction=&GetUserInterfaceUV}.Normalized UV) {materialFunction=&MF_UI_UVScaling}.Scaled to Fit) {name=Scale To Fit, nodeColor=<rgba 0 0 0 1>}
  n2 = Constant() {r=0.6}
  MaterialFunctionCall!(UVs=@n1, Size=@n2, CornerRoundness=Constant() {r=0.0001}, KeepAspectRatio, PixelSize) {materialFunction=/FortUILibrary/MaterialFunctions/Shapes/SDFs/MF_SDF_Box}
  n3 = Constant()
  n4 = TextureCoordinate()
  n5 = LinearInterpolate(
   A=MaterialFunctionCall(UVs=@n1, Size=@n2, CornerRoundness=@n3) {materialFunction=&MF_UI_Box}.SDF
   B=MaterialFunctionCall(
    UVs=MaterialFunctionCall!(UVs=@n4) {materialFunction=/FortUILibrary/MaterialFunctions/Basics/MF_UI_9SliceUVs}.UVs
    Size=Constant() {r=0.75}
    CornerRoundness=@n3
    OutlineThickness=Constant() {r=0.05}
   ) {materialFunction=&MF_UI_Box}.SDF
   Alpha=ScalarParameter() {parameterName=Defense bg}
  )
  n6 = Multiply(A=MaterialFunctionCall!(DistanceField=@n5, Threshold, Thickness=Constant() {r=0.05}, Alignment=Constant() {r=3}, Smoothing) {materialFunction=/FortUILibrary/MaterialFunctions/Shapes/Modifiers/MF_UI_ShapeOutline}.Outline)
  n7 = Multiply(A=MaterialFunctionCall(DistanceField=@n5, Threshold=Constant() {r=0.06}) {materialFunction=&MF_UI_AntiAliasing}.Inner) {constB=0.6}
  NamedRerouteDeclaration(None=Max(
    A=NamedRerouteDeclaration(None=@n6) {name=Mask_BoxOutline, nodeColor=<rgba 0 0 0 0.75>}
    B=NamedRerouteDeclaration(None=@n7) {name=Mask_BoxShape, nodeColor=<rgba 0 0 0 0.75>}
   )) {name=Mask_Box, nodeColor=<rgba 0 0 0 1>}
  n8 = NamedRerouteUsage()
  n9 = ScalarParameter() {defaultValue=0.6, parameterName=Icon Scale}
  n10 = TextureObjectParameter() {parameterName=Texture Input, samplerType=SAMPLERTYPE_LinearGrayscale, texture=/TidalCraneSharedActivitiesUI/Textures/T_UI_Icon_Hive}
  n11 = NamedRerouteUsage()
  n12 = LinearInterpolate(
   A=TextureSample(
    UVs=MaterialFunctionCall(UVs=@n8, Texture Scale=@n9) {materialFunction=&ScaleUVsByCenter}.UVs
    Tex=@n10
   ) {samplerType=SAMPLERTYPE_LinearGrayscale, texture=/TidalCraneSharedActivitiesUI/Textures/T_UI_Icon_BaseDefense_Bug}.RGB
   B=TextureSample(UVs=MaterialFunctionCall(UVs=@n11, Texture Scale=@n9) {materialFunction=&ScaleUVsByCenter}.UVs) {samplerType=SAMPLERTYPE_LinearGrayscale, texture=/TidalCraneSharedActivitiesUI/Textures/T_Hive_Checkmark}.RGB
   Alpha=ScalarParameter() {parameterName=CheckMark Switch}
  )
  NamedRerouteDeclaration(None=MaterialFunctionCall(DistanceField=@n12, Threshold=Constant() {r=0.75}) {materialFunction=&MF_UI_AntiAliasing}.Outer) {name=Mask_Texture, nodeColor=<rgba 0 0 0 0.75>}
  NamedRerouteDeclaration(None=MaterialFunctionCall(DistanceField=@n12, Threshold=Constant() {r=0.5}) {materialFunction=&MF_UI_AntiAliasing}.Outer) {name=Mask_TextureShadow, nodeColor=<rgba 0 0 0 0.75>}
  n13 = MaterialFunctionCall(DistanceField=@n5, Thickness=Constant() {r=0.05}, Alignment=Constant() {r=3}) {materialFunction=&MF_UI_ShapeOutline}
  n14 = MaterialFunctionCall(DistanceField=@n13.SDFOutline, Threshold=Constant() {r=0.1}, Smoothing=Constant() {r=0.2}) {materialFunction=&MF_UI_AntiAliasing}
  NamedRerouteDeclaration(None=Multiply(A=Power(Base=@n14.Inner) {constExponent=7}) {constB=0.1}) {name=Glow Alpha, nodeColor=<rgba 0 0 0 1>}
  OneMinus(None=@n13.SDFOutline)
  n15 = NamedRerouteUsage()
  n16 = NamedRerouteUsage()
  StaticSwitchParameter(True=@n15, False=Add(A=@n15, B=Multiply(A=@n16) {constB=0.3})) {parameterName=RemoveIcon}
  n17 = NamedRerouteUsage()
  n18 = NamedRerouteUsage()
  n19 = LinearInterpolate(
   A=Constant3Vector() {constant=<rgba 0 0 0 1>}
   B=LinearInterpolate(
    A=VectorParameter() {group=Color, parameterName=Outline Color, defaultValue=<rgba 0.0578 0.5647 0.2705 1>}.RGB
    B=Constant() {r=1}
    Alpha=ScalarParameter() {parameterName=CheckMark Switch}
   )
   Alpha=Multiply(A=Add(A=@n17, B=@n18)) {constB=0.5}
  )
  n20 = ScalarParameter() {parameterName=CheckMark Switch}
  n21 = LinearInterpolate(A=VectorParameter() {group=Color, parameterName=Icon Outline Color, defaultValue=<rgba 0 0.1719 0.1059 1>}.RGB, Alpha=@n20) {constB=0}
  n22 = LinearInterpolate(A=VectorParameter() {group=Color, parameterName=Icon Color, defaultValue=<rgba 0 0.6867 0.4233 1>}.RGB, B=Constant() {r=1}, Alpha=@n20)
  n23 = Multiply(A=NamedRerouteUsage()) {constB=0.8}
  StaticSwitchParameter(True=@n19, False=LinearInterpolate(A=@n19, B=LinearInterpolate(A=@n21, B=@n22, Alpha=@n23), Alpha=NamedRerouteUsage())) {parameterName=RemoveIcon}
 ]
}
