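# Additive UI material for a soft, rounded-square glow, named for Hive POI in the shared activities UI.
# A rounded box SDF, drawn on fitted and center-scaled UVs, is inverted, sharpened with a power and saturated into a falloff.
# That falloff, dimmed, blends between the Color A and Color B parameters, red to orange by default.
# The Checkmark Color scalar then lerps the result toward white.
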Material /TidalCraneSharedActivitiesUI/Materials/M_UI_HivePOI_SquareGlow
graph {
 properties {bAutomaticallySetUsageInEditor=false, blendMode=BLEND_Additive, materialDomain=MD_UI, usedShadingModels=Default Lit}
 nodes [
  n1 = MaterialFunctionCall(
   UVs=MaterialFunctionCall(
    UVs=MaterialFunctionCall() {materialFunction=&MF_UI_UVScaling}.Scaled to Fit
    Texture Scale=Constant() {r=0.5}
   ) {materialFunction=&ScaleUVsByCenter}.UVs
   Size=Constant2Vector() {g=1, r=1}
   CornerRoundness=Constant() {r=0.2}
   KeepAspectRatio=StaticBool() {value=true}
  ) {materialFunction=&MF_UI_Box}
  n2 = Saturate(None=Power(Base=OneMinus(None=@n1.SDF)) {constExponent=3})
  LinearInterpolate(
   A=LinearInterpolate(
    A=VectorParameter() {parameterName=Color A, defaultValue=<rgba 0.3698 0.0278 0 1>}.RGB
    B=VectorParameter() {parameterName=Color B, defaultValue=<rgba 1 0.3251 0 1>}.RGB
    Alpha=Multiply(A=@n2) {constB=0.4}
   )
   Alpha=ScalarParameter() {parameterName=Checkmark Color}
  )
 ]
}
