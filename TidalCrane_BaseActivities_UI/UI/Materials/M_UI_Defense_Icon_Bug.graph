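# UI material for the base defense bug icon, using alpha-composite blending.
# A grayscale distance-field texture, scaled about its center by Icon Scale, feeds two anti-aliasing thresholds that produce an icon mask and a wider shadow mask.
# Nested lerps blend outline and icon colors through reroutes.
# CheckMark Switch shifts the outline color toward black and the icon color toward white.
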Material /TidalCrane_BaseActivities_UI/UI/Materials/M_UI_Defense_Icon_Bug
graph {
 properties {bAutomaticallySetUsageInEditor=false, blendMode=BLEND_AlphaComposite, materialDomain=MD_UI, usedShadingModels=Default Lit}
 nodes [
  n1 = VectorParameter() {group=Color, parameterName=Icon Outline Color, defaultValue=<rgba 0.1119 0.0203 0.0015 1>}
  n2 = ScalarParameter() {parameterName=CheckMark Switch}
  n3 = VectorParameter() {group=Color, parameterName=Icon Color, defaultValue=<rgba 1 0.2789 0 1>}
  n4 = Constant() {r=1}
  n5 = NamedRerouteUsage()
  LinearInterpolate(
   B=LinearInterpolate(A=LinearInterpolate(A=@n1.RGB, Alpha=@n2) {constB=0}, B=LinearInterpolate(A=@n3.RGB, B=@n4, Alpha=@n2), Alpha=Multiply(A=@n5) {constB=0.8})
   Alpha=NamedRerouteUsage()
  )
  Multiply(A=NamedRerouteUsage()) {constB=0.3}
  n6 = TextureObjectParameter() {
   parameterName=Texture Input
   samplerType=SAMPLERTYPE_LinearGrayscale
   texture=/TidalCraneSharedActivitiesUI/Textures/T_UI_Icon_BaseDefense_Bug
  }
  n7 = TextureSample(
   UVs=MaterialFunctionCall(UVs=NamedRerouteUsage(), Texture Scale=ScalarParameter() {defaultValue=1, parameterName=Icon Scale}) {materialFunction=&ScaleUVsByCenter}.UVs
   Tex=@n6
  ) {samplerType=SAMPLERTYPE_LinearGrayscale, texture=&T_UI_Icon_BaseDefense_Bug}
  NamedRerouteDeclaration(None=MaterialFunctionCall(DistanceField=@n7.RGB, Threshold=Constant() {r=0.75}) {materialFunction=&MF_UI_AntiAliasing}.Outer) {name=Mask_Texture, nodeColor=<rgba 0 0 0 0.75>}
  NamedRerouteDeclaration(None=MaterialFunctionCall(DistanceField=@n7.RGB, Threshold=Constant() {r=0.5}) {materialFunction=&MF_UI_AntiAliasing}.Outer) {name=Mask_TextureShadow, nodeColor=<rgba 0 0 0 0.75>}
 ]
}
