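# HUD widget for Tidal Crane base defense events, showing an objective message, bugs-killed counter and progress bar.
# Defense data from the spawn manager updates the counter with kill flashes; a looping timer eases the bar and glow materials' Progress upward.
# Reaching the bug total plays a result message, then an outro after a delay; failure goes straight to the outro.
# The widget starts collapsed and resizes for touch and tablet screens.
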WidgetBlueprint /TidalCrane_BaseActivities_UI/UI/WBP_TidalCrane_BaseActivities_UI_Defense
animations [
 Anim_Intro {
  binding_count=6
  length_seconds=1.8834
  object=&self:Anim_Intro
  sequence {
   bindings [
    Text_EventMessage {
     id=#1
     possessed_class=CommonTextBlock
     tracks [
      {
       type=MovieSceneFloatTrack
       display_name=Render Opacity
       sections [{type=MovieSceneFloatSection, channels [None {type=MovieSceneScriptingFloatChannel, keys [[16.9987,0],[24.3333,0],[24.9987,1]]}]}]
      }
      {
       type=MovieScene2DTransformTrack
       sections [{
         type=MovieScene2DTransformSection
         end_frame=25
         start_frame=21
         channels [
          Translation.X {type=MovieSceneScriptingFloatChannel, keys [[21.3327,-30],[25,0],[27.6667,0]]}
          Translation.Y {type=MovieSceneScriptingFloatChannel, const [0,1]}
          Angle {type=MovieSceneScriptingFloatChannel, const [0,1]}
          Scale.X {type=MovieSceneScriptingFloatChannel, const [1,1]}
          Scale.Y {type=MovieSceneScriptingFloatChannel, const [1,1]}
          Shear.X {type=MovieSceneScriptingFloatChannel, const [0,1]}
          Shear.Y {type=MovieSceneScriptingFloatChannel, const [0,1]}
         ]
        }]
      }
     ]
    }
    Image_TextFX {
     id=#2
     possessed_class=Image
     tracks [
      {
       type=MovieSceneWidgetMaterialTrack
       sections [{
         type=MovieSceneParameterSection
         channels [
          Blinky X Movement {type=MovieSceneScriptingFloatChannel, keys [[22.6667,0.266],[25.6667,-0.29]]}
          Mask {type=MovieSceneScriptingFloatChannel, keys [[22.6667,0],[25.6667,1]]}
         ]
        }]
      }
      {
       type=MovieSceneFloatTrack
       display_name=Render Opacity
       sections [{
         type=MovieSceneFloatSection
         channels [None {type=MovieSceneScriptingFloatChannel, keys [[15.3333,0],[17,0.5],[22.6667,0.5],[25.3333,1],[27.6667,1],[28.6667,0]]}]
        }]
      }
     ]
    }
    Overlay_BugKillCount {
     id=#3
     possessed_class=Overlay
     tracks [
      {
       type=MovieSceneFloatTrack
       display_name=Render Opacity
       sections [{type=MovieSceneFloatSection, channels [None {type=MovieSceneScriptingFloatChannel, keys [[15,0],[18,1]]}]}]
      }
      {
       type=MovieScene2DTransformTrack
       sections [{
         type=MovieScene2DTransformSection
         end_frame=16
         start_frame=15
         channels [
          Translation.X {type=MovieSceneScriptingFloatChannel, keys [[15,-20],[16.7143,0]]}
          Translation.Y {type=MovieSceneScriptingFloatChannel, const [0,1]}
          Angle {type=MovieSceneScriptingFloatChannel, const [0,1]}
          Scale.X {type=MovieSceneScriptingFloatChannel, const [1,1]}
          Scale.Y {type=MovieSceneScriptingFloatChannel, const [1,1]}
          Shear.X {type=MovieSceneScriptingFloatChannel, const [0,1]}
          Shear.Y {type=MovieSceneScriptingFloatChannel, const [0,1]}
         ]
        }]
      }
     ]
    }
    Overlay_Objective {
     id=#4
     possessed_class=Overlay
     tracks [{
       type=MovieSceneFloatTrack
       display_name=Render Opacity
       sections [{type=MovieSceneFloatSection, end_frame=0, start_frame=0, channels [None {type=MovieSceneScriptingFloatChannel, const [1,1]}]}]
      }]
    }
    Image_ProgressBar {
     id=#5
     possessed_class=Image
     tracks [
      {
       type=MovieSceneWidgetMaterialTrack
       sections [{
         type=MovieSceneParameterSection
         channels [
          Double Dots Opacity {type=MovieSceneScriptingFloatChannel, keys [[12.3333,0],[12.6667,1],[13.3333,0],[14.6667,1],[15.3333,0.8],[16.6667,0],[18.3333,1]]}
          Double Dots Size {type=MovieSceneScriptingFloatChannel, thinned_from=4, keys [[12.6667,1],[14.3333,1],[14.6667,0]]}
          Outer Accent X Movement {type=MovieSceneScriptingFloatChannel, keys [[3.3333,1],[7.3333,0]]}
          Outer Bar Flash {type=MovieSceneScriptingFloatChannel, keys [[0,1],[19.6667,0.5],[26.6667,0]]}
          Progress Bar Reveal {type=MovieSceneScriptingFloatChannel, keys [[6,0],[11.3333,1]]}
          Progress Bar Flash {type=MovieSceneScriptingFloatChannel, keys [[6,1],[11.3333,1],[21.3333,0]]}
          Outer Frame Opacity {type=MovieSceneScriptingFloatChannel, keys [[0,0],[17.3333,0],[18.3333,1],[19.6667,0.5]]}
          Outer Accents Opacity {type=MovieSceneScriptingFloatChannel, const [1,1]}
          Progress Opacity {type=MovieSceneScriptingFloatChannel, keys [[21.6667,0],[26.3333,1]]}
          Border Thickness {type=MovieSceneScriptingFloatChannel, keys [[18,1],[25,0]]}
         ]
        }]
      }
      {
       type=MovieSceneFloatTrack
       display_name=Render Opacity
       sections [{type=MovieSceneFloatSection, channels [None {type=MovieSceneScriptingFloatChannel, const [1,1]}]}]
      }
     ]
    }
    Image_Glow {
     id=#6
     possessed_class=Image
     tracks [{
       type=MovieSceneFloatTrack
       display_name=Render Opacity
       sections [{type=MovieSceneFloatSection, channels [None {type=MovieSceneScriptingFloatChannel, keys [[25,0],[37.6667,0.1]]}]}]
      }]
    }
   ]
   tracks [{
     type=MovieSceneAudioTrack
     display_name=Audio
     sections [{
       type=MovieSceneAudioSection
       end_frame=20
       start_frame=0
       channels [
        Volume {type=MovieSceneScriptingFloatChannel, default=1}
        Pitch {type=MovieSceneScriptingFloatChannel, default=1}
        AttachActor {type=MovieSceneScriptingActorReferenceChannel, default="<Struct 'MovieSceneObjectBindingID' (0x4696e99f760) {}>"}
       ]
      }]
    }]
  }
 }
 Anim_Outro {
  binding_count=3
  length_seconds=1.15
  object=&self:Anim_Outro
  sequence {bindings [
    Image_ProgressBar {
     id=#7
     possessed_class=Image
     tracks [{
       type=MovieSceneWidgetMaterialTrack
       sections [{
         type=MovieSceneParameterSection
         channels [
          Double Dots Opacity {type=MovieSceneScriptingFloatChannel, keys [[0,1],[3,0],[3.3333,1],[4,0],[5.3333,1],[6,0.8],[7.3333,0],[9,0]]}
          Double Dots Size {type=MovieSceneScriptingFloatChannel, thinned_from=5, keys [[0,0],[3.3333,1],[5,1],[5.3333,0]]}
          Outer Accent X Movement {type=MovieSceneScriptingFloatChannel, keys [[0,0],[7.3333,0],[13,1],[15,1]]}
          Outer Bar Flash {type=MovieSceneScriptingFloatChannel, keys [[0,0],[7.3333,0],[13,1],[15,1],[23,0]]}
          Progress Bar Reveal {type=MovieSceneScriptingFloatChannel, keys [[0,1],[7,1],[9,0]]}
          Progress Bar Flash {type=MovieSceneScriptingFloatChannel, const [0,1]}
          Outer Frame Opacity {type=MovieSceneScriptingFloatChannel, keys [[0,1],[5,1],[7,0]]}
          Outer Accents Opacity {type=MovieSceneScriptingFloatChannel, thinned_from=4, keys [[0,1],[15,1],[23,0]]}
          Progress Opacity {type=MovieSceneScriptingFloatChannel, keys [[0,1],[3,0]]}
          Progress {type=MovieSceneScriptingFloatChannel, const [1,1]}
         ]
        }]
      }]
    }
    Overlay_Objective {
     id=#8
     possessed_class=Overlay
     tracks [{
       type=MovieSceneFloatTrack
       display_name=Render Opacity
       sections [{type=MovieSceneFloatSection, channels [None {type=MovieSceneScriptingFloatChannel, keys [[0,1],[5,0]]}]}]
      }]
    }
    Image_Glow {
     id=#9
     possessed_class=Image
     tracks [
      {
       type=MovieSceneWidgetMaterialTrack
       sections [{type=MovieSceneParameterSection, end_frame=17, start_frame=0, channels [Progress {type=MovieSceneScriptingFloatChannel, const [1,1]}]}]
      }
      {
       type=MovieSceneFloatTrack
       display_name=Render Opacity
       sections [{type=MovieSceneFloatSection, end_frame=5, start_frame=0, channels [None {type=MovieSceneScriptingFloatChannel, keys [[0,0.1],[5,0]]}]}]
      }
      {
       type=MovieScene2DTransformTrack
       sections [{
         type=MovieScene2DTransformSection
         channels [
          Translation.X {type=MovieSceneScriptingFloatChannel, const [0,1]}
          Translation.Y {type=MovieSceneScriptingFloatChannel, const [-32,1]}
          Angle {type=MovieSceneScriptingFloatChannel, const [0,1]}
          Scale.X {type=MovieSceneScriptingFloatChannel, keys [[0,1],[5,0]]}
          Scale.Y {type=MovieSceneScriptingFloatChannel, const [1,1]}
          Shear.X {type=MovieSceneScriptingFloatChannel, const [0,1]}
          Shear.Y {type=MovieSceneScriptingFloatChannel, const [0,1]}
         ]
        }]
      }
     ]
    }
   ]}
 }
 Anim_Result {
  binding_count=2
  length_seconds=0.5834
  object=&self:Anim_Result
  sequence {bindings [
    Text_EventMessage {
     id=#10
     possessed_class=CommonTextBlock
     tracks [
      {
       type=MovieSceneFloatTrack
       display_name=Render Opacity
       sections [{type=MovieSceneFloatSection, channels [None {type=MovieSceneScriptingFloatChannel, keys [[-0.0007,0],[7.334,0],[7.9993,1]]}]}]
      }
      {
       type=MovieScene2DTransformTrack
       sections [{
         type=MovieScene2DTransformSection
         end_frame=8
         start_frame=4
         channels [
          Translation.X {type=MovieSceneScriptingFloatChannel, keys [[4.3333,-30],[8.0007,0],[10.6673,0]]}
          Translation.Y {type=MovieSceneScriptingFloatChannel, const [0,1]}
          Angle {type=MovieSceneScriptingFloatChannel, const [0,1]}
          Scale.X {type=MovieSceneScriptingFloatChannel, const [1,1]}
          Scale.Y {type=MovieSceneScriptingFloatChannel, const [1,1]}
          Shear.X {type=MovieSceneScriptingFloatChannel, const [0,1]}
          Shear.Y {type=MovieSceneScriptingFloatChannel, const [0,1]}
         ]
        }]
      }
     ]
    }
    Image_TextFX {
     id=#11
     possessed_class=Image
     tracks [
      {
       type=MovieSceneWidgetMaterialTrack
       sections [{
         type=MovieSceneParameterSection
         channels [
          Blinky X Movement {type=MovieSceneScriptingFloatChannel, keys [[5.6673,0.266],[8.6667,-0.28]]}
          Mask {type=MovieSceneScriptingFloatChannel, keys [[5.6673,0],[8.6673,1]]}
         ]
        }]
      }
      {
       type=MovieSceneFloatTrack
       display_name=Render Opacity
       sections [{
         type=MovieSceneFloatSection
         channels [None {type=MovieSceneScriptingFloatChannel, thinned_from=6, keys [[0,0],[2.6667,1],[10.6673,1],[11.6673,0]]}]
        }]
      }
     ]
    }
   ]}
 }
 Anim_BugKill {
  binding_count=1
  length_seconds=0.3
  object=&self:Anim_BugKill
  sequence {bindings [Image_IconGlow {
     id=#12
     possessed_class=Image
     tracks [
      {
       type=MovieSceneFloatTrack
       display_name=Render Opacity
       sections [{
         type=MovieSceneFloatSection
         end_frame=6
         start_frame=0
         channels [None {type=MovieSceneScriptingFloatChannel, keys [[0,0.2],[3.4283,1],[6,0.1]]}]
        }]
      }
      {
       type=MovieScene2DTransformTrack
       sections [{
         type=MovieScene2DTransformSection
         end_frame=6
         start_frame=0
         channels [
          Translation.X {type=MovieSceneScriptingFloatChannel, const [0,1]}
          Translation.Y {type=MovieSceneScriptingFloatChannel, const [0,1]}
          Angle {type=MovieSceneScriptingFloatChannel, const [90,1]}
          Scale.X {type=MovieSceneScriptingFloatChannel, const [2,1]}
          Scale.Y {type=MovieSceneScriptingFloatChannel, keys [[0,0],[3.4283,7],[6,5]]}
          Shear.X {type=MovieSceneScriptingFloatChannel, const [0,1]}
          Shear.Y {type=MovieSceneScriptingFloatChannel, const [0,1]}
         ]
        }]
      }
     ]
    }]}
 }
 Anim_Bomber {
  binding_count=1
  length_seconds=0.25
  object=&self:Anim_Bomber
  sequence {bindings [Image_ProgressBar {
     id=#13
     possessed_class=Image
     tracks [{
       type=MovieSceneWidgetMaterialTrack
       sections [{
         type=MovieSceneParameterSection
         end_frame=5
         start_frame=0
         channels [BomberBugKill? {type=MovieSceneScriptingFloatChannel, keys [[0,0],[5,1]]}]
        }]
      }]
    }]}
 }
]
graphs [
 BindAnimationCallbacks {nodes [
   BindAnimationCallbacks() {is=title K2Node_FunctionEntry}
   > BindToAnimationFinished(Animation=Get Anim_Outro() -> (Anim_Outro:Widget Animation Object Reference), Delegate=Create Event())
  ]}
 EventGraph {nodes [
   OnEventComplete() {is=title K2Node_CustomEvent}
   n1 = Get Anim_Result() -> (Anim_Result:Widget Animation Object Reference)
   > n2 = Branch(Condition=OR Boolean(A=IsAnimationPlaying(InAnimation=@n1), B=Get EventIsComplete() -> (EventIsComplete:bool)))
   Set EventIsComplete(execute=@n2.else, EventIsComplete:bool=true) -> (Output_Get:bool)
   > PlayAnimationForward(InAnimation=Get Anim_Result())
   > "SetText (Text)"(self=Get Text_EventMessage() -> (Text_EventMessage:Common Text Object Reference), InText=Get EventCompleteText() -> (EventCompleteText:Text))
   > Delay(Duration=3)
   > PlayOutroAnimation()
   OnEventFailed() {is=title K2Node_CustomEvent}
   > Branch(Condition="Not Equal (Enum)"!(
     A:ESlateVisibility Enum=GetVisibility!(self:Widget Object Reference) -> (ReturnValue:ESlateVisibility Enum=Visible)
     B:ESlateVisibility Enum=Collapsed
    ) -> (ReturnValue:bool))
   > Set EventIsComplete(EventIsComplete=true)
   > PlayOutroAnimation()
   Event OnInitialized()
   > BindAnimationCallbacks() {is=title K2Node_CallFunction}
   > SetVisibility(InVisibility=Collapsed)
   > SetTouchLayout()
   > n3 = GetDynamicMaterial(self=Get Image_ProgressBar() -> (Image_ProgressBar:Image Object Reference))
   > Set DynamicMaterial_ProgressBar(DynamicMaterial_ProgressBar:Material Instance Dynamic Object Reference=@n3) -> (Output_Get:Material Instance Dynamic Object Reference)
   > n4 = GetDynamicMaterial(self=Get Image_Glow() -> (Image_Glow:Image Object Reference))
   > Set DynamicMaterial_GlowBar(DynamicMaterial_GlowBar:Material Instance Dynamic Object Reference=@n4) -> (Output_Get:Material Instance Dynamic Object Reference)
  ]}
 EventHasFinished {nodes [
   n1 = EventHasFinished!() -> (then, Base:Spawn Manager Tidal Crane Defense Object Reference) {is=title K2Node_FunctionEntry}
   n2 = Get CurrentBase() -> (CurrentBase:Spawn Manager Tidal Crane Defense Object Reference)
   > Return Node!(execute, EventIsComplete:bool=AND Boolean(A=Get EventIsComplete(), B="Equal (Object)"(A=@n1, B=@n2)))
  ]}
 OnBomberBugKilled {nodes [
   OnBomberBugKilled() {is=title K2Node_FunctionEntry}
   > PlayAnimation(InAnimation=Get Anim_Bomber() -> (Anim_Bomber:Widget Animation Object Reference))
  ]}
 PlayOutroAnimation {nodes [
   PlayOutroAnimation() {is=title K2Node_FunctionEntry}
   > PlayAnimation(InAnimation=Get Anim_Outro())
  ]}
 ResetWidgetHUD {nodes [
   ResetWidgetHUD() {is=title K2Node_FunctionEntry}
   > "SetText (Text)"(self=Get Text_EventMessage(), InText=Get EventMessage() -> (EventMessage:Text))
  ]}
 SetDefenseData {nodes [
   n1 = SetDefenseData!() -> (then, Base:Spawn Manager Tidal Crane Defense Object Reference, Defense Data:Tidal Crane Base Defense Data Structure) {is=title K2Node_FunctionEntry}
   > Set CurrentBase(CurrentBase:Spawn Manager Tidal Crane Defense Object Reference=@n1) -> (Output_Get:Spawn Manager Tidal Crane Defense Object Reference)
   > n2 = Sequence()
   > UpdateBugsCounter!(execute, self, Defense Data:Tidal Crane Base Defense Data Structure=@n1.Defense Data, UpdatingData:bool=false) -> (then)
   > SetProgressBar()
   Set MaxNPCs(
    execute=@n2.then_1
    MaxNPCs:int=Break Tidal Crane Base Defense Data!(TidalCrane_Base_DefenseData:"Tidal Crane Base Defense Data Structure (by ref)"=@n1.Defense Data) -> (
     SpawnManager_14_#14:Spawn Manager Tidal Crane Defense Object Reference
     MaxNPCs_4_#15:int
    ) {is=title K2Node_BreakStruct}.MaxNPCs_4_#15
   ) -> (Output_Get:int=0)
   > Set EventIsComplete(EventIsComplete=false)
  ]}
 SetProgressBar {nodes [
   SetProgressBar() {is=title K2Node_FunctionEntry}
   n1 = float / float(
    A="To Float (Integer)"(InInt=Get BugsKilled() -> (BugsKilled:int=0))
    B="To Float (Integer)"(InInt=Get MaxBugsCount() -> (MaxBugsCount:int=0))
   )
   > SetScalarParameterValue(
    self=Get DynamicMaterial_ProgressBar() -> (DynamicMaterial_ProgressBar:Material Instance Dynamic Object Reference)
    ParameterName=Progress
    Value=@n1
   )
   > SetScalarParameterValue(
    self=Get DynamicMaterial_GlowBar() -> (DynamicMaterial_GlowBar:Material Instance Dynamic Object Reference)
    ParameterName=Progress
    Value=@n1
   )
   > Set CurrentProgressBar(CurrentProgressBar:double=@n1) -> (Output_Get:double)
  ]}
 SetTouchLayout {nodes [
   SetTouchLayout() {is=title K2Node_FunctionEntry}
   > Branch(Condition=IsUsingTouch(PlayerController=GetOwningPlayer()))
   > SetRenderScale!(
    execute
    self:Widget Object Reference=Get Overlay_BugKillCount() -> (Overlay_BugKillCount:Overlay Object Reference)
    Scale:Vector 2D Structure=(X=1.2,Y=1.2)
   ) -> (then)
   > SetHeightOverride(self=Get SizeBox_Progress() -> (SizeBox_Progress:Size Box Object Reference), InHeightOverride=85)
   > SetPadding!(
    execute
    self:Vertical Box Slot Object Reference=SlotAsVerticalBoxSlot(Widget=Get SizeBox_Progress())
    InPadding_Left:float=0
    InPadding_Top:float=-32
    InPadding_Right:float=0
    InPadding_Bottom:float=0
   ) -> (then)
   > n1 = Branch(Condition=IsTabletAspectRatio!() -> (ReturnValue:bool))
   > SetWidthOverride(self=Get SizeBox_Progress(), InWidthOverride=450)
   SetWidthOverride(execute=@n1.else, self=Get SizeBox_Progress(), InWidthOverride=600)
  ]}
 SetTrackerVisibility {nodes [
   SetTrackerVisibility() {is=title K2Node_FunctionEntry}
   > SetVisibility(InVisibility=Collapsed)
  ]}
 ShowEventHUD {nodes [
   n1 = ShowEventHUD!() -> (then, Payload:Tidal Crane Base Timer Event Structure) {is=title K2Node_FunctionEntry}
   > ResetWidgetHUD() {is=title K2Node_CallFunction}
   > Branch(Condition="Equal (Enum)"(
     A:TidalCrane_Bases_Enum Enum=Break Tidal Crane Base Timer Event!(TidalCrane_Base_TimerEvent:"Tidal Crane Base Timer Event Structure (by ref)"=@n1) -> (
      SpawnManagerActor_5_#16:Spawn Manager Tidal Crane Defense Object Reference
      BasePhase_8_#17:TidalCrane_Bases_Enum Enum
      StartTime_11_#18:double=~
      EndTime_13_#19:double=~
      TotalDuration_15_#20:double=~
     ) {is=title K2Node_BreakStruct}.BasePhase_8_#17
     B:TidalCrane_Bases_Enum Enum=NewEnumerator2
    ))
   > PlayAnimationForward(InAnimation=Get Anim_Intro() -> (Anim_Intro:Widget Animation Object Reference))
   > SetVisibility(InVisibility=HitTestInvisible)
  ]}
 UpdateBugsCounter {nodes [
   n1 = UpdateBugsCounter!() -> (then, Defense Data:Tidal Crane Base Defense Data Structure, UpdatingData:bool=~) {is=title K2Node_FunctionEntry}
   > n2 = Branch(Condition=@n1.UpdatingData)
   > n3 = Branch(Condition="Not Equal (Integer)"(A=Get BombersKilled() -> (BombersKilled:int=0)))
   > OnBomberBugKilled() {is=title K2Node_CallFunction}
   > n4 = Set BombersKilled(BombersKilled:int=0) -> (Output_Get:int=0)
   Set SwarmersKilled(execute=[@n2.else, @n3.else, @n4.then], SwarmersKilled:int=0) -> (Output_Get:int=0)
   n5 = Break Tidal Crane Base Defense Data!(TidalCrane_Base_DefenseData=@n1) -> (
    SpawnManager_14_#14
    EventState_10_#21:TidalCrane_Bases_Enum Enum
    AliveNPCs_2_#22:int
    MaxNPCs_4_#15
    BugsKilled_6_#23:int
    MaxBugs_27_#24:int
    CurrTimerStartTime_22_#25:double=~
    CurrTimerDuration_24_#26:double=~
   )
   > n6 = Set BugsKilled(BugsKilled:int=@n5.BugsKilled_6_#23) -> (Output_Get:int=0)
   > n7 = Set MaxBugsCount(MaxBugsCount:int=@n5.MaxBugs_27_#24) -> (Output_Get:int=0)
   n8 = SelectInt(A=@n7, B=@n6, bPickA="integer >= integer"(A=Get BugsKilled() -> (BugsKilled=0), B=Get MaxBugsCount() -> (MaxBugsCount=0)))
   n9 = Get MaxBugsCount() -> (MaxBugsCount=0)
   > "SetText (Text)"(
    self=Get Text_BugsKilledCounter() -> (Text_BugsKilledCounter:Common Text Object Reference)
    InText=Format Text!(
     Format="LOCTEXT(\"{BugsKilled} / {TotalBugs}\")"
     BugsKilled:Text="To Text (Integer)"(Value=@n8)
     TotalBugs:Text=To Text!(Value:int=@n9) -> (ReturnValue:Text) {is=title K2Node_GenericToText}
    ) -> (Result)
   )
   > PlayAnimation(InAnimation=Get Anim_BugKill() -> (Anim_BugKill:Widget Animation Object Reference))
  ]}
 UpdateDefenseData {nodes [
   n1 = UpdateDefenseData!() -> (then, Defense Data:Tidal Crane Base Defense Data Structure) {is=title K2Node_FunctionEntry}
   > UpdateBugsCounter(Defense Data=@n1, UpdatingData=true) {is=title K2Node_CallFunction}
   > n2 = Set Timer by Event(Delegate=Create Event(), Time=0.01, bLooping=true, bMaxOncePerFrame=true)
   > Set TimerHandleStructure_Progress(TimerHandleStructure_Progress:Timer Handle Structure=@n2) -> (Output_Get:Timer Handle Structure)
   > Branch(Condition="integer >= integer"(A=Get BugsKilled() -> (BugsKilled=0), B=Get MaxBugsCount() -> (MaxBugsCount=0)))
   > OnEventComplete() {is=title K2Node_CallFunction}
  ]}
 UpdateProgressBar {nodes [
   UpdateProgressBar() {is=title K2Node_FunctionEntry}
   n1 = "To Float (Integer)"(InInt=Get BugsKilled() -> (BugsKilled=0))
   n2 = "To Float (Integer)"(InInt=Get MaxBugsCount() -> (MaxBugsCount=0))
   > n3 = Branch(Condition="float >= float"(A=Get CurrentProgressBar() -> (CurrentProgressBar:double), B=float / float(A=@n1, B=@n2)))
   > Clear and Invalidate Timer by Handle(Handle=Get TimerHandleStructure_Progress() -> (TimerHandleStructure_Progress:Timer Handle Structure))
   n4 = Get BugsKilled() -> (BugsKilled=0)
   n5 = Get MaxBugsCount() -> (MaxBugsCount=0)
   > Set CurrentProgressBar(CurrentProgressBar=float / float(A="To Float (Integer)"(InInt=@n4), B="To Float (Integer)"(InInt=@n5)))
   n6 = Get CurrentProgressBar()
   SetScalarParameterValue(execute=@n3.else, self=Get DynamicMaterial_ProgressBar(), ParameterName=Progress, Value=@n6)
   > SetScalarParameterValue(self=Get DynamicMaterial_GlowBar(), ParameterName=Progress, Value=@n6)
   > Set CurrentProgressBar(CurrentProgressBar=float + float(A=Get CurrentProgressBar(), B=0.001))
  ]}
]
variables [
 BombersKilled {type=integer}
 BugsKilled {type=integer}
 CurrentBase {type=SpawnManager_TidalCrane_Defense_C, default=null}
 CurrentNPCs {type="array<FortPlayerPawn>", default []}
 CurrentProgressBar {type=number}
 DynamicMaterial_GlowBar {type=MaterialInstanceDynamic}
 DynamicMaterial_ProgressBar {type=MaterialInstanceDynamic}
 EventCompleteText {type=string, default="BUGS ELIMINATED!"}
 EventFailedText {type=string, default="BUGS ELIMINATED!"}
 EventIsComplete {type=boolean}
 EventMaxDurationTime {type=number}
 EventMessage {type=string, default="Defend the Outpost!"}
 EventTime {type=number}
 MaxBugsCount {type=integer}
 MaxNPCs {type=integer}
 NPCsAlive {type=integer}
 SwarmersKilled {type=integer}
 TimerHandleStructure {type=TimerHandle}
 TimerHandleStructure_Progress {type=TimerHandle}
]
widgets {
 info {inheritedWidgetCount=0, parentClass=&CommonUserWidget, rootWidgetClass=&VerticalBox, widgetCount=21}
 widgets [
  VerticalBox_Container {widgetClassPath=&VerticalBox, bIsVariable=true, widget=VerticalBox_Container}
  SizeBox_Progress {
   widgetClassPath=&SizeBox
   bIsVariable=true
   parent=VerticalBox_Container
   widget=SizeBox_Progress
   properties {bOverride_WidthOverride=true, heightOverride=70, widthOverride=700}
   slot {class=VerticalBoxSlot}
  }
  VerticalBox_60 {widgetClassPath=&VerticalBox, parent=SizeBox_Progress, widget=VerticalBox_60, slot {class=SizeBoxSlot}}
  Overlay_Progress {widgetClassPath=&Overlay, parent=VerticalBox_60, widget=Overlay_Progress, slot {class=VerticalBoxSlot}}
  Image_ProgressBar {
   widgetClassPath=&Image
   bIsVariable=true
   parent=Overlay_Progress
   widget=Image_ProgressBar
   properties {
    visibility=SelfHitTestInvisible
    brush {
     bIsDynamicallyLoaded=false
     drawAs=Image
     imageType=FullColor
     mirroring=NoMirror
     resourceName=None
     resourceObject=/TidalCraneSharedActivitiesUI/Materials/M_UI_Reclamation_Progress
     tiling=NoTile
     imageSize=<700 70>
     margin=<ltrb 0 0 0 0>
     outlineSettings {
      bUseBrushTransparency=false
      roundingType=HalfHeightRadius
      width=0
      color {colorUseRule=UseColor_Specified, specifiedColor=<rgba 0 0 0 0>}
      cornerRadii=<0 0 0 0>
     }
     tintColor {colorUseRule=UseColor_Specified, specifiedColor=<rgba 1 1 1 1>}
     uVRegion {bIsValid=false, max=<0 0>, min=<0 0>}
    }
   }
   slot {class=OverlaySlot}
  }
  Image_Glow {
   widgetClassPath=&Image
   bIsVariable=true
   parent=Overlay_Progress
   widget=Image_Glow
   properties {
    renderOpacity=0.1
    visibility=SelfHitTestInvisible
    brush {
     bIsDynamicallyLoaded=false
     drawAs=Image
     imageType=FullColor
     mirroring=NoMirror
     resourceName=None
     resourceObject=/TidalCraneSharedActivitiesUI/Materials/MI_UI_Reclamation_Progress_Glow
     tiling=NoTile
     imageSize=<30 32>
     margin=<ltrb 0 0 0 0>
     outlineSettings {
      bUseBrushTransparency=false
      roundingType=HalfHeightRadius
      width=0
      color {colorUseRule=UseColor_Specified, specifiedColor=<rgba 0 0 0 0>}
      cornerRadii=<0 0 0 0>
     }
     tintColor {colorUseRule=UseColor_Specified, specifiedColor=<rgba 1 1 1 1>}
     uVRegion {bIsValid=false, max=<0 0>, min=<0 0>}
    }
    renderTransform {angle=0, scale=<1 1>, shear=<0 0>, translation=<0 -32>}
   }
   slot {class=OverlaySlot, properties {horizontalAlignment=HAlign_Fill, verticalAlignment=VAlign_Fill}}
  }
  Overlay_Objective {
   widgetClassPath=&Overlay
   bIsVariable=true
   parent=VerticalBox_60
   widget=Overlay_Objective
   slot {class=VerticalBoxSlot, properties {verticalAlignment=VAlign_Center, padding=<ltrb 20 0 20 0>}}
  }
  HorizontalBox_1 {
   widgetClassPath=&HorizontalBox
   parent=Overlay_Objective
   widget=HorizontalBox_1
   slot {class=OverlaySlot, properties {verticalAlignment=VAlign_Center}}
  }
  Overlay_BugKillCount {
   widgetClassPath=&Overlay
   bIsVariable=true
   parent=HorizontalBox_1
   widget=Overlay_BugKillCount
   slot {class=HorizontalBoxSlot, properties {horizontalAlignment=HAlign_Center, verticalAlignment=VAlign_Center}}
  }
  Spacer_301 {widgetClassPath=&Spacer, parent=Overlay_BugKillCount, widget=Spacer_301, slot {class=OverlaySlot}}
  Image_Bg {
   widgetClassPath=&Image
   parent=Overlay_BugKillCount
   widget=Image_Bg
   properties {
    visibility=SelfHitTestInvisible
    brush {
     bIsDynamicallyLoaded=false
     drawAs=Image
     imageType=FullColor
     mirroring=NoMirror
     resourceName=None
     resourceObject=/TidalCrane_BaseActivities_UI/UI/Materials/MI_UI_Defense_Icon_Bg
     tiling=NoTile
     imageSize=<20 20>
     margin=<ltrb 0 0 0 0>
     outlineSettings {
      bUseBrushTransparency=false
      roundingType=HalfHeightRadius
      width=0
      color {colorUseRule=UseColor_Specified, specifiedColor=<rgba 0 0 0 0>}
      cornerRadii=<0 0 0 0>
     }
     tintColor {colorUseRule=UseColor_Specified, specifiedColor=<rgba 1 1 1 1>}
     uVRegion {bIsValid=false, max=<0 0>, min=<0 0>}
    }
   }
   slot {class=OverlaySlot, properties {horizontalAlignment=HAlign_Fill, verticalAlignment=VAlign_Fill}}
  }
  HorizontalBox_BugCount {
   widgetClassPath=&HorizontalBox
   parent=Overlay_BugKillCount
   widget=HorizontalBox_BugCount
   properties {bOverride_Cursor=true}
   slot {class=OverlaySlot, properties {horizontalAlignment=HAlign_Center, verticalAlignment=VAlign_Center, padding=<ltrb 6 0 6 0>}}
  }
  Overlay_1 {widgetClassPath=&Overlay, parent=HorizontalBox_BugCount, widget=Overlay_1, slot {class=HorizontalBoxSlot}}
  Image_Bugs {
   widgetClassPath=&Image
   parent=Overlay_1
   widget=Image_Bugs
   properties {
    visibility=SelfHitTestInvisible
    brush {
     bIsDynamicallyLoaded=false
     drawAs=Image
     imageType=FullColor
     mirroring=NoMirror
     resourceName=None
     resourceObject=/TidalCrane_BaseActivities_UI/UI/Materials/M_UI_Defense_Icon_Bug
     tiling=NoTile
     imageSize=<40 40>
     margin=<ltrb 0 0 0 0>
     outlineSettings {
      bUseBrushTransparency=false
      roundingType=HalfHeightRadius
      width=0
      color {colorUseRule=UseColor_Specified, specifiedColor=<rgba 0 0 0 0>}
      cornerRadii=<0 0 0 0>
     }
     tintColor {colorUseRule=UseColor_Specified, specifiedColor=<rgba 1 1 1 1>}
     uVRegion {bIsValid=false, max=<0 0>, min=<0 0>}
    }
   }
   slot {class=OverlaySlot}
  }
  Spacer_57 {
   widgetClassPath=&Spacer
   parent=HorizontalBox_BugCount
   widget=Spacer_57
   properties {size=<4 30>}
   slot {class=HorizontalBoxSlot, properties {horizontalAlignment=HAlign_Left, verticalAlignment=VAlign_Center}}
  }
  Text_BugsKilledCounter {
   widgetClassPath=&CommonTextBlock
   bIsVariable=true
   parent=HorizontalBox_BugCount
   widget=Text_BugsKilledCounter
   properties {
    text=30/30
    colorAndOpacity {colorUseRule=UseColor_Specified, specifiedColor=<rgba 1 0.1848 0 1>}
    font {
     bForceMonospaced=false
     bMaterialIsStencil=false
     fontMaterial=null
     fontObject=&BurbankSmall
     letterSpacing=0
     monospacedWidth=1
     size=16.5
     skewAmount=0
     typefaceFontName=Black
     outlineSettings {
      bApplyOutlineToDropShadows=false
      bMiteredCorners=false
      bSeparateFillAlpha=false
      outlineMaterial=null
      outlineSize=2
      outlineColor=<rgba 0 0 0 0.5>
     }
    }
    shadowOffset=<0 0>
   }
   slot {class=HorizontalBoxSlot, properties {horizontalAlignment=HAlign_Left, verticalAlignment=VAlign_Center}}
  }
  Image_IconGlow {
   widgetClassPath=&Image
   bIsVariable=true
   parent=Overlay_BugKillCount
   widget=Image_IconGlow
   properties {
    renderOpacity=0.2
    visibility=SelfHitTestInvisible
    brush {
     bIsDynamicallyLoaded=false
     drawAs=Image
     imageType=FullColor
     mirroring=NoMirror
     resourceName=None
     resourceObject=/TidalCraneSharedActivitiesUI/Materials/M_UI_HivePOI_SquareGlow
     tiling=NoTile
     imageSize=<30 30>
     margin=<ltrb 0 0 0 0>
     outlineSettings {
      bUseBrushTransparency=false
      roundingType=HalfHeightRadius
      width=0
      color {colorUseRule=UseColor_Specified, specifiedColor=<rgba 0 0 0 0>}
      cornerRadii=<0 0 0 0>
     }
     tintColor {colorUseRule=UseColor_Specified, specifiedColor=<rgba 1 1 1 1>}
     uVRegion {bIsValid=false, max=<0 0>, min=<0 0>}
    }
    renderTransform {angle=90, scale=<2 5>, shear=<0 0>, translation=<0 0>}
   }
   slot {class=OverlaySlot, properties {horizontalAlignment=HAlign_Fill, verticalAlignment=VAlign_Fill}}
  }
  Spacer_BugCountRight {
   widgetClassPath=&Spacer
   bIsVariable=true
   parent=HorizontalBox_1
   widget=Spacer_BugCountRight
   properties {size=<30 30>}
   slot {class=HorizontalBoxSlot, properties {horizontalAlignment=HAlign_Left, verticalAlignment=VAlign_Center}}
  }
  Overlay_52 {
   widgetClassPath=&Overlay
   parent=HorizontalBox_1
   widget=Overlay_52
   slot {class=HorizontalBoxSlot, properties {verticalAlignment=VAlign_Center}}
  }
  Text_EventMessage {
   widgetClassPath=&CommonTextBlock
   bIsVariable=true
   parent=Overlay_52
   widget=Text_EventMessage
   properties {
    mobileFontSizeMultiplier=1.2
    text="Defend the Outpost!"
    textTransformPolicy=ToUpper
    wrapTextAt=550
    colorAndOpacity {colorUseRule=UseColor_Specified, specifiedColor=<rgba 1 1 1 0.9>}
    font {
     bForceMonospaced=false
     bMaterialIsStencil=false
     fontMaterial=null
     fontObject=&BurbankSmall
     letterSpacing=0
     monospacedWidth=1
     size=19.5
     skewAmount=0
     typefaceFontName=Black
     outlineSettings {
      bApplyOutlineToDropShadows=false
      bMiteredCorners=false
      bSeparateFillAlpha=false
      outlineMaterial=&M-FontShadow
      outlineSize=3
      outlineColor=<rgba 0 0 0 0.5>
     }
    }
    shadowOffset=<0 0>
   }
   slot {class=OverlaySlot}
  }
  Image_TextFX {
   widgetClassPath=&Image
   parent=Overlay_52
   widget=Image_TextFX
   properties {
    renderOpacity=0
    visibility=SelfHitTestInvisible
    brush {
     bIsDynamicallyLoaded=false
     drawAs=Image
     imageType=FullColor
     mirroring=NoMirror
     resourceName=None
     resourceObject=/TidalCraneSharedActivitiesUI/Materials/M_UI_TextBlink
     tiling=NoTile
     imageSize=<32 32>
     margin=<ltrb 0 0 0 0>
     outlineSettings {
      bUseBrushTransparency=false
      roundingType=HalfHeightRadius
      width=0
      color {colorUseRule=UseColor_Specified, specifiedColor=<rgba 0 0 0 0>}
      cornerRadii=<0 0 0 0>
     }
     tintColor {colorUseRule=UseColor_Specified, specifiedColor=<rgba 1 1 1 1>}
     uVRegion {bIsValid=false, max=<0 0>, min=<0 0>}
    }
    renderTransform {angle=0, scale=<2 2>, shear=<0 0>, translation=<0 0>}
   }
   slot {class=OverlaySlot, properties {horizontalAlignment=HAlign_Fill, verticalAlignment=VAlign_Fill}}
  }
 ]
}
