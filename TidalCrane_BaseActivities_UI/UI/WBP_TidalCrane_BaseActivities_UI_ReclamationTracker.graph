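# CommonUserWidget icon tracking one Reclamation objective, shown as an NPC hive or spawner with matching glows.
# It starts hidden until PlayIntroAnimation fades and slides it in; SetObjectiveType swaps its three materials.
# UpdateObjectiveState plays the checkmark clear animation once, guarded by ObjectiveIsCleared.
# SetObjectiveState starts that animation past its end when cleared, or rewinds it and resets the flag otherwise, without setting the flag when cleared.
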WidgetBlueprint /TidalCrane_BaseActivities_UI/UI/WBP_TidalCrane_BaseActivities_UI_ReclamationTracker
animations [
 Anim_ClearAnimation {
  binding_count=3
  length_seconds=0.3
  object=&self:Anim_ClearAnimation
  sequence {bindings [
    Image_SquareGlow {
     id=#1
     possessed_class=Image
     tracks [
      {
       type=MovieSceneWidgetMaterialTrack
       sections [{
         type=MovieSceneParameterSection
         end_frame=5
         start_frame=-1
         channels [Checkmark Color {type=MovieSceneScriptingFloatChannel, thinned_from=4, keys [[-1,0],[3,0],[5,1]]}]
        }]
      }
      {
       type=MovieSceneFloatTrack
       display_name=Render Opacity
       sections [{type=MovieSceneFloatSection, end_frame=5, start_frame=3, channels [None {type=MovieSceneScriptingFloatChannel, keys [[3,0.3],[5,0.1]]}]}]
      }
     ]
    }
    Image_Icon {
     id=#2
     possessed_class=Image
     tracks [{
       type=MovieSceneWidgetMaterialTrack
       sections [{
         type=MovieSceneParameterSection
         end_frame=5
         start_frame=0
         channels [
          CheckMark Switch {type=MovieSceneScriptingFloatChannel, keys [[0,0],[2.0833,0],[2.5,1],[5,1]]}
          Icon Scale {type=MovieSceneScriptingFloatChannel, keys [[0,0.6],[2.0833,0.75],[5,0.55]]}
         ]
        }]
      }]
    }
    Image_IconGlow {
     id=#3
     possessed_class=Image
     tracks [
      {
       type=MovieScene2DTransformTrack
       sections [{
         type=MovieScene2DTransformSection
         end_frame=5
         start_frame=0
         channels [
          Translation.X {type=MovieSceneScriptingFloatChannel, const [0,1]}
          Translation.Y {type=MovieSceneScriptingFloatChannel, const [0,1]}
          Angle {type=MovieSceneScriptingFloatChannel, const [0,1]}
          Scale.X {type=MovieSceneScriptingFloatChannel, const [1.7,1]}
          Scale.Y {type=MovieSceneScriptingFloatChannel, keys [[0,2],[2.5,4],[5,0]]}
          Shear.X {type=MovieSceneScriptingFloatChannel, const [0,1]}
          Shear.Y {type=MovieSceneScriptingFloatChannel, const [0,1]}
         ]
        }]
      }
      {
       type=MovieSceneFloatTrack
       display_name=Render Opacity
       sections [{
         type=MovieSceneFloatSection
         end_frame=5
         start_frame=-1
         channels [None {type=MovieSceneScriptingFloatChannel, keys [[-1,0],[0,0],[4.1667,1],[5,0]]}]
        }]
      }
     ]
    }
   ]}
 }
 Anim_Intro {
  binding_count=2
  length_seconds=0.1
  object=&self:Anim_Intro
  sequence {
   bindings [
    WBP_TidalCrane_BaseActivities_UI_ReclamationTracker_C_4 {
     id=#4
     possessed_class=WBP_TidalCrane_BaseActivities_UI_ReclamationTracker_C
     tracks [{
       type=MovieSceneFloatTrack
       display_name=Render Opacity
       sections [{type=MovieSceneFloatSection, end_frame=6, start_frame=0, channels [None {type=MovieSceneScriptingFloatChannel, keys [[0,0],[6,1]]}]}]
      }]
    }
    Overlay {
     id=#5
     possessed_class=Overlay
     tracks [{
       type=MovieScene2DTransformTrack
       sections [{
         type=MovieScene2DTransformSection
         end_frame=6
         start_frame=0
         channels [
          Translation.X {type=MovieSceneScriptingFloatChannel, keys [[0,-35],[6,0]]}
          Translation.Y {type=MovieSceneScriptingFloatChannel, const [0,1]}
          Angle {type=MovieSceneScriptingFloatChannel, const [0,1]}
          Scale.X {type=MovieSceneScriptingFloatChannel, const [1,1]}
          Scale.Y {type=MovieSceneScriptingFloatChannel, const [1,1]}
          Shear.X {type=MovieSceneScriptingFloatChannel, const [0,1]}
          Shear.Y {type=MovieSceneScriptingFloatChannel, const [0,1]}
         ]
        }]
      }]
    }
   ]
   display_rate {numerator=60}
  }
 }
]
graphs [
 EventGraph {nodes [
   Event Construct()
   > SetVisibility(InVisibility=Hidden)
  ]}
 PlayIntroAnimation {nodes [
   PlayIntroAnimation() {is=title K2Node_FunctionEntry}
   > PlayAnimation(InAnimation=Get Anim_Intro() -> (Anim_Intro:Widget Animation Object Reference))
   > SetVisibility(InVisibility=HitTestInvisible)
  ]}
 SetObjectiveState {nodes [
   n1 = SetObjectiveState!() -> (then, ObjectiveCleared:bool=~) {is=title K2Node_FunctionEntry}
   > n2 = Branch(Condition=@n1)
   > PlayAnimation(InAnimation=Get Anim_ClearAnimation() -> (Anim_ClearAnimation:Widget Animation Object Reference), StartAtTime=1)
   PlayAnimationTimeRange!(
    execute=@n2.else
    self:User Widget Object Reference
    InAnimation:Widget Animation Object Reference=Get Anim_ClearAnimation()
    StartAtTime:float=0
    EndAtTime:float=0
    NumLoopsToPlay:int=1
    PlayMode:EUMGSequencePlayMode Enum
    PlaybackSpeed:float=1
    bRestoreState:bool=false
   ) -> (then, ReturnValue:Widget Animation Handle Structure)
   > Set ObjectiveIsCleared(ObjectiveIsCleared:bool=false) -> (Output_Get:bool)
  ]}
 SetObjectiveType {nodes [
   n1 = SetObjectiveType!() -> (then, isSpawnerType:bool=~) {is=title K2Node_FunctionEntry}
   > SetBrushFromMaterial(
    self=Get Image_Icon() -> (Image_Icon:Image Object Reference)
    Material=Select(
     Option 0:Material Interface Object Reference=Get Material_NPC_Icon() -> (Material_NPC_Icon:Material Instance Object Reference)
     Option 1:Material Interface Object Reference=Get Material_Spawner_Icon() -> (Material_Spawner_Icon:Material Instance Object Reference)
     Index=@n1
    ) -> (ReturnValue:Material Interface Object Reference)
   )
   > SetBrushFromMaterial(
    self=Get Image_SquareGlow() -> (Image_SquareGlow:Image Object Reference)
    Material=Select(
     Option 0=Get Material_NPC_Glow() -> (Material_NPC_Glow:Material Instance Object Reference)
     Option 1=Get Material_Spawner_Glow() -> (Material_Spawner_Glow:Material Instance Object Reference)
     Index=@n1
    )
   )
   > SetBrushFromMaterial(
    self=Get Image_IconGlow() -> (Image_IconGlow:Image Object Reference)
    Material=Select(
     Option 0=Get Material_NPC_HardEdgeGlow() -> (Material_NPC_HardEdgeGlow:Material Instance Object Reference)
     Option 1=Get Material_Spawner_HardEdgeGlow() -> (Material_Spawner_HardEdgeGlow:Material Instance Object Reference)
     Index=@n1
    )
   )
  ]}
 UpdateObjectiveState {nodes [
   UpdateObjectiveState() {is=title K2Node_FunctionEntry}
   > n1 = Branch(Condition=Get ObjectiveIsCleared() -> (ObjectiveIsCleared:bool))
   PlayAnimation(execute=@n1.else, InAnimation=Get Anim_ClearAnimation())
   > Set ObjectiveIsCleared(ObjectiveIsCleared=true)
  ]}
]
variables [
 Material_NPC_Glow {type=MaterialInstance, default=/TidalCraneSharedActivitiesUI/Materials/MI_UI_HivePOI_SquareGlow}
 Material_NPC_HardEdgeGlow {type=MaterialInstance, default=/TidalCraneSharedActivitiesUI/Materials/MI_UI_HardEdgedGlow}
 Material_NPC_Icon {type=MaterialInstance, default=/TidalCrane_BaseActivities_UI/UI/Materials/MI_UI_Reclamation_Hive_Icon}
 Material_Spawner_Glow {type=MaterialInstance, default=/TidalCraneSharedActivitiesUI/Materials/MI_UI_HivePOI_SquareGlow_Purple}
 Material_Spawner_HardEdgeGlow {type=MaterialInstance, default=/TidalCraneSharedActivitiesUI/Materials/MI_UI_HardEdgedGlow_Purple}
 Material_Spawner_Icon {type=MaterialInstance, default=/TidalCrane_BaseActivities_UI/UI/Materials/MI_UI_Reclamation_Spawner_Icon}
 ObjectiveIsCleared {type=boolean}
]
widgets {
 info {inheritedWidgetCount=0, parentClass=&CommonUserWidget, rootWidgetClass=&Overlay, widgetCount=4}
 widgets [
  Overlay {widgetClassPath=&Overlay, widget=Overlay}
  Image_Icon {
   widgetClassPath=&Image
   bIsVariable=true
   parent=Overlay
   widget=Image_Icon
   properties {
    visibility=SelfHitTestInvisible
    brush {
     bIsDynamicallyLoaded=false
     drawAs=Image
     imageType=FullColor
     mirroring=NoMirror
     resourceName=None
     resourceObject=&MI_UI_Reclamation_Hive_Icon
     tiling=NoTile
     imageSize=<40 40>
     margin=<ltrb 0 0 0 0>
     outlineSettings {
      bUseBrushTransparency=false
      roundingType=HalfHeightRadius
      width=0
      color {colorUseRule=UseColor_Specified, specifiedColor=<rgba 0 0 0 0>}
      cornerRadii=<0 0 0 0>
     }
     tintColor {colorUseRule=UseColor_Specified, specifiedColor=<rgba 1 1 1 1>}
     uVRegion {bIsValid=false, max=<0 0>, min=<0 0>}
    }
   }
   slot {class=OverlaySlot, properties {horizontalAlignment=HAlign_Center, verticalAlignment=VAlign_Center}}
  }
  Image_SquareGlow {
   widgetClassPath=&Image
   bIsVariable=true
   parent=Overlay
   widget=Image_SquareGlow
   properties {
    renderOpacity=0.3
    visibility=SelfHitTestInvisible
    brush {
     bIsDynamicallyLoaded=false
     drawAs=Image
     imageType=FullColor
     mirroring=NoMirror
     resourceName=None
     resourceObject=&MI_UI_HivePOI_SquareGlow
     tiling=NoTile
     imageSize=<32 32>
     margin=<ltrb 0 0 0 0>
     outlineSettings {
      bUseBrushTransparency=false
      roundingType=HalfHeightRadius
      width=0
      color {colorUseRule=UseColor_Specified, specifiedColor=<rgba 0 0 0 0>}
      cornerRadii=<0 0 0 0>
     }
     tintColor {colorUseRule=UseColor_Specified, specifiedColor=<rgba 1 1 1 1>}
     uVRegion {bIsValid=false, max=<0 0>, min=<0 0>}
    }
    renderTransform {angle=0, scale=<1.7 1.7>, shear=<0 0>, translation=<0 0>}
   }
   slot {class=OverlaySlot, properties {horizontalAlignment=HAlign_Center, verticalAlignment=VAlign_Center}}
  }
  Image_IconGlow {
   widgetClassPath=&Image
   bIsVariable=true
   parent=Overlay
   widget=Image_IconGlow
   properties {
    renderOpacity=0
    visibility=SelfHitTestInvisible
    brush {
     bIsDynamicallyLoaded=false
     drawAs=Image
     imageType=FullColor
     mirroring=NoMirror
     resourceName=None
     resourceObject=&MI_UI_HardEdgedGlow
     tiling=NoTile
     imageSize=<20 32>
     margin=<ltrb 0 0 0 0>
     outlineSettings {
      bUseBrushTransparency=false
      roundingType=HalfHeightRadius
      width=0
      color {colorUseRule=UseColor_Specified, specifiedColor=<rgba 0 0 0 0>}
      cornerRadii=<0 0 0 0>
     }
     tintColor {colorUseRule=UseColor_Specified, specifiedColor=<rgba 1 1 1 1>}
     uVRegion {bIsValid=false, max=<0 0>, min=<0 0>}
    }
    renderTransform {angle=0, scale=<1.55 2.5>, shear=<0 0>, translation=<0 0>}
   }
   slot {class=OverlaySlot, properties {horizontalAlignment=HAlign_Center, verticalAlignment=VAlign_Center}}
  }
 ]
}
