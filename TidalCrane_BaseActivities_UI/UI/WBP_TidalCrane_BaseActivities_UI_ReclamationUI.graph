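# HUD widget for an outpost reclamation event, showing a progress bar, message, and trackers for NPCs to free and spawners to destroy.
# SetEventData rebuilds trackers when the base or objective count changes, otherwise calling UpdateEventHUD when completed counts rise.
# A looping timer drains the bar's dynamic material parameters toward the remaining fraction, triggering the completion message and outro at zero.
# Sequencer animations drive material reveals, with staggered tracker intros and touch-specific sizing.
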WidgetBlueprint /TidalCrane_BaseActivities_UI/UI/WBP_TidalCrane_BaseActivities_UI_ReclamationUI
animations [
 Anim_Intro {
  binding_count=5
  length_seconds=1.8834
  object=&self:Anim_Intro
  sequence {
   bindings [
    Image_Glow {
     id=#1
     possessed_class=Image
     tracks [{
       type=MovieSceneFloatTrack
       display_name=Render Opacity
       sections [{type=MovieSceneFloatSection, channels [None {type=MovieSceneScriptingFloatChannel, keys [[75,0],[113,0.1]]}]}]
      }]
    }
    Image_ProgressBar {
     id=#2
     possessed_class=Image
     tracks [
      {
       type=MovieSceneFloatTrack
       display_name=Render Opacity
       sections [{type=MovieSceneFloatSection, channels [None {type=MovieSceneScriptingFloatChannel, const [1,1]}]}]
      }
      {
       type=MovieSceneWidgetMaterialTrack
       sections [{
         type=MovieSceneParameterSection
         channels [
          Double Dots Opacity {type=MovieSceneScriptingFloatChannel, keys [[37,0],[38,1],[40,0],[44,1],[46,0.8],[50,0],[55,1]]}
          Double Dots Size {type=MovieSceneScriptingFloatChannel, thinned_from=4, keys [[38,1],[43,1],[44,0]]}
          Outer Accent X Movement {type=MovieSceneScriptingFloatChannel, keys [[10,1],[22,0]]}
          Outer Bar Flash {type=MovieSceneScriptingFloatChannel, keys [[0,1],[59,0.5],[80,0]]}
          Progress Bar Reveal {type=MovieSceneScriptingFloatChannel, keys [[18,0],[34,1]]}
          Progress Bar Flash {type=MovieSceneScriptingFloatChannel, keys [[18,1],[34,1],[64,0]]}
          Outer Frame Opacity {type=MovieSceneScriptingFloatChannel, keys [[0,0],[52,0],[55,1],[59,0.5]]}
          Outer Accents Opacity {type=MovieSceneScriptingFloatChannel, const [1,1]}
          Progress Opacity {type=MovieSceneScriptingFloatChannel, keys [[65,0],[79,1]]}
          Border Thickness {type=MovieSceneScriptingFloatChannel, keys [[54,1],[75,0]]}
         ]
        }]
      }
     ]
    }
    Text_ReclamationMessage {
     id=#3
     possessed_class=CommonTextBlock
     tracks [
      {
       type=MovieSceneFloatTrack
       display_name=Render Opacity
       sections [{type=MovieSceneFloatSection, channels [None {type=MovieSceneScriptingFloatChannel, keys [[50.996,0],[73,0],[74.996,1]]}]}]
      }
      {
       type=MovieScene2DTransformTrack
       sections [{
         type=MovieScene2DTransformSection
         end_frame=75
         start_frame=63
         channels [
          Translation.X {type=MovieSceneScriptingFloatChannel, keys [[63.998,-30],[75,0],[83,0]]}
          Translation.Y {type=MovieSceneScriptingFloatChannel, const [0,1]}
          Angle {type=MovieSceneScriptingFloatChannel, const [0,1]}
          Scale.X {type=MovieSceneScriptingFloatChannel, const [1,1]}
          Scale.Y {type=MovieSceneScriptingFloatChannel, const [1,1]}
          Shear.X {type=MovieSceneScriptingFloatChannel, const [0,1]}
          Shear.Y {type=MovieSceneScriptingFloatChannel, const [0,1]}
         ]
        }]
      }
     ]
    }
    HorizontalBox_Objective {
     id=#4
     possessed_class=HorizontalBox
     tracks [{
       type=MovieSceneFloatTrack
       display_name=Render Opacity
       sections [{type=MovieSceneFloatSection, end_frame=113, start_frame=61, channels [None {type=MovieSceneScriptingFloatChannel, const [1,2]}]}]
      }]
    }
    Image_TextFX {
     id=#5
     possessed_class=Image
     tracks [
      {
       type=MovieSceneWidgetMaterialTrack
       sections [{
         type=MovieSceneParameterSection
         channels [
          Blinky X Movement {type=MovieSceneScriptingFloatChannel, keys [[68,0.266],[77,-0.29]]}
          Mask {type=MovieSceneScriptingFloatChannel, keys [[68,0],[77,1]]}
         ]
        }]
      }
      {
       type=MovieSceneFloatTrack
       display_name=Render Opacity
       sections [{type=MovieSceneFloatSection, channels [None {type=MovieSceneScriptingFloatChannel, keys [[46,0],[51,0.5],[68,0.5],[76,1],[83,1],[86,0]]}]}]
      }
     ]
    }
   ]
   display_rate {numerator=60}
   tracks [{
     type=MovieSceneEventTrack
     sections [{
       type=MovieSceneEventTriggerSection
       channels [None {type=MovieSceneScriptingEventChannel, const ["<Struct 'MovieSceneEvent' (0x4696ee30b40) {}>", 1]}]
      }]
    }]
  }
 }
 Anim_Outro {
  binding_count=3
  length_seconds=1.15
  object=&self:Anim_Outro
  sequence {bindings [
    Image_ProgressBar {
     id=#6
     possessed_class=Image
     tracks [{
       type=MovieSceneWidgetMaterialTrack
       sections [{
         type=MovieSceneParameterSection
         channels [
          Double Dots Opacity {type=MovieSceneScriptingFloatChannel, keys [[0,1],[3,0],[3.3333,1],[4,0],[5.3333,1],[6,0.8],[7.3333,0],[9,0]]}
          Double Dots Size {type=MovieSceneScriptingFloatChannel, thinned_from=5, keys [[0,0],[3.3333,1],[5,1],[5.3333,0]]}
          Outer Accent X Movement {type=MovieSceneScriptingFloatChannel, keys [[0,0],[7.3333,0],[13,1],[15,1]]}
          Outer Bar Flash {type=MovieSceneScriptingFloatChannel, keys [[0,0],[7.3333,0],[13,1],[15,1],[23,0]]}
          Progress Bar Reveal {type=MovieSceneScriptingFloatChannel, keys [[0,1],[7,1],[9,0]]}
          Progress Bar Flash {type=MovieSceneScriptingFloatChannel, const [0,1]}
          Outer Frame Opacity {type=MovieSceneScriptingFloatChannel, keys [[0,1],[5,1],[7,0]]}
          Outer Accents Opacity {type=MovieSceneScriptingFloatChannel, thinned_from=4, keys [[0,1],[15,1],[23,0]]}
          Progress Opacity {type=MovieSceneScriptingFloatChannel, keys [[0,1],[3,0]]}
          Progress {type=MovieSceneScriptingFloatChannel, const [0,1]}
         ]
        }]
      }]
    }
    HorizontalBox_Objective {
     id=#7
     display_name=HorizontalBox_1
     possessed_class=HorizontalBox
     tracks [{
       type=MovieSceneFloatTrack
       display_name=Render Opacity
       sections [{type=MovieSceneFloatSection, channels [None {type=MovieSceneScriptingFloatChannel, keys [[0,1],[5,0]]}]}]
      }]
    }
    Image_Glow {
     id=#8
     possessed_class=Image
     tracks [
      {
       type=MovieSceneFloatTrack
       display_name=Render Opacity
       sections [{type=MovieSceneFloatSection, end_frame=5, start_frame=0, channels [None {type=MovieSceneScriptingFloatChannel, keys [[0,0.1],[5,0]]}]}]
      }
      {
       type=MovieSceneWidgetMaterialTrack
       sections [{type=MovieSceneParameterSection, end_frame=17, start_frame=0, channels [Progress {type=MovieSceneScriptingFloatChannel, const [0,1]}]}]
      }
      {
       type=MovieScene2DTransformTrack
       sections [{
         type=MovieScene2DTransformSection
         channels [
          Translation.X {type=MovieSceneScriptingFloatChannel, const [0,1]}
          Translation.Y {type=MovieSceneScriptingFloatChannel, const [-32,1]}
          Angle {type=MovieSceneScriptingFloatChannel, const [0,1]}
          Scale.X {type=MovieSceneScriptingFloatChannel, keys [[0,1],[5,0]]}
          Scale.Y {type=MovieSceneScriptingFloatChannel, const [1,1]}
          Shear.X {type=MovieSceneScriptingFloatChannel, const [0,1]}
          Shear.Y {type=MovieSceneScriptingFloatChannel, const [0,1]}
         ]
        }]
      }
     ]
    }
   ]}
 }
 Anim_Completed {
  binding_count=2
  length_seconds=0.5834
  object=&self:Anim_Completed
  sequence {
   bindings [
    Image_TextFX {
     id=#9
     possessed_class=Image
     tracks [
      {
       type=MovieSceneWidgetMaterialTrack
       sections [{
         type=MovieSceneParameterSection
         channels [
          Blinky X Movement {type=MovieSceneScriptingFloatChannel, keys [[17.002,0.266],[26,-0.28]]}
          Mask {type=MovieSceneScriptingFloatChannel, keys [[17.002,0],[26.002,1]]}
         ]
        }]
      }
      {
       type=MovieSceneFloatTrack
       display_name=Render Opacity
       sections [{
         type=MovieSceneFloatSection
         channels [None {type=MovieSceneScriptingFloatChannel, thinned_from=6, keys [[0,0],[8,1],[32.002,1],[35.002,0]]}]
        }]
      }
     ]
    }
    Text_ReclamationMessage {
     id=#10
     possessed_class=CommonTextBlock
     tracks [
      {
       type=MovieScene2DTransformTrack
       sections [{
         type=MovieScene2DTransformSection
         end_frame=24
         start_frame=13
         channels [
          Translation.X {type=MovieSceneScriptingFloatChannel, keys [[13,-30],[24.002,0],[32.002,0]]}
          Translation.Y {type=MovieSceneScriptingFloatChannel, const [0,1]}
          Angle {type=MovieSceneScriptingFloatChannel, const [0,1]}
          Scale.X {type=MovieSceneScriptingFloatChannel, const [1,1]}
          Scale.Y {type=MovieSceneScriptingFloatChannel, const [1,1]}
          Shear.X {type=MovieSceneScriptingFloatChannel, const [0,1]}
          Shear.Y {type=MovieSceneScriptingFloatChannel, const [0,1]}
         ]
        }]
      }
      {
       type=MovieSceneFloatTrack
       display_name=Render Opacity
       sections [{type=MovieSceneFloatSection, channels [None {type=MovieSceneScriptingFloatChannel, keys [[-0.002,0],[22.002,0],[23.998,1]]}]}]
      }
     ]
    }
   ]
   display_rate {numerator=60}
  }
 }
 Anim_Progress {
  binding_count=2
  length_seconds=1
  object=&self:Anim_Progress
  sequence {bindings [
    Image_ProgressBar {id=#11, possessed_class=Image, tracks []}
    Image_Glow {
     id=#12
     possessed_class=Image
     tracks [{
       type=MovieSceneWidgetMaterialTrack
       sections [{
         type=MovieSceneParameterSection
         end_frame=20
         start_frame=0
         channels [Progress {type=MovieSceneScriptingFloatChannel, keys [[0,0],[20,1]]}]
        }]
      }]
    }
   ]}
 }
]
graphs [
 BindAnimationCallbacks {nodes [
   BindAnimationCallbacks() {is=title K2Node_FunctionEntry}
   > BindToAnimationFinished(Animation=Get Anim_Outro() -> (Anim_Outro:Widget Animation Object Reference), Delegate=Create Event())
   > BindToAnimationFinished(Animation=Get Anim_Outro(), Delegate=Create Event())
   Get Anim_Completed() -> (Anim_Completed:Widget Animation Object Reference)
  ]}
 CheckProgressBarOffset {nodes [
   CheckProgressBarOffset() {is=title K2Node_FunctionEntry}
   n1 = float / float(A=Get CurrentProgress() -> (CurrentProgress:double), B:int=Get CurrentMaxObjectivesCount() -> (CurrentMaxObjectivesCount:int=0))
   n2 = float / float(A=Get CurrentProgress(), B=Get CurrentMaxObjectivesCount() -> (CurrentMaxObjectivesCount=0))
   > n3 = Branch(Condition=AND Boolean(A="float < float"(A=@n1, B=0.5), B="Not Equal (Float)"(A=@n2)))
   > Set ProgressOffset(ProgressOffset:double=-0.015) -> (Output_Get:double)
   n4 = Get CurrentMaxObjectivesCount() -> (CurrentMaxObjectivesCount=0)
   n5 = Get CurrentProgress()
   n6 = Branch(execute=@n3.else, Condition="float > float"(A=float / float(A=@n5, B=@n4), B=0.5))
   > Set ProgressOffset(ProgressOffset=0.015)
   Set ProgressOffset(execute=@n6.else, ProgressOffset=0)
  ]}
 CreateObjectiveTrackers {nodes [
   n1 = CreateObjectiveTrackers!() -> (then, MaxObjectives:int, MaxNpcs:int) {is=title K2Node_FunctionEntry}
   n2 = GetAllEntries!(self:Dynamic Entry Box Base Object Reference=Get DynamicEntryBox_ObjectiveTracker() -> (DynamicEntryBox_ObjectiveTracker:Dynamic Entry Box Object Reference)) -> (ReturnValue:Array of User Widget Object References)
   > n3 = Branch(Condition="Equal (Integer)"(A=@n1, B=Length(TargetArray:Array of User Widget Object References=@n2)))
   Reset!(execute=@n3.else, self:Dynamic Entry Box Object Reference=Get DynamicEntryBox_ObjectiveTracker(), bDeleteWidgets:bool=false) -> (then)
   > n4 = For Loop(FirstIndex=1, LastIndex=@n1)
   > n5 = Create Entry!(execute, self:Dynamic Entry Box Object Reference=Get DynamicEntryBox_ObjectiveTracker()) -> (then, ReturnValue:User Widget Object Reference)
   > n6 = Cast To WBP_TidalCrane_BaseActivities_UI_ReclamationTracker!(execute, Object=@n5) -> (
    then
    CastFailed
    AsWBP Tidal Crane Base Activities UI Reclamation Tracker:WBP Tidal Crane Base Activities UI Reclamation Tracker Object Reference
   ) {is=title K2Node_DynamicCast}
   n7 = "integer > integer"(A=@n4, B=@n1.MaxNpcs)
   > SetObjectiveType!(execute, self:WBP Tidal Crane Base Activities UI Reclamation Tracker Object Reference=@n6, isSpawnerType:bool=@n7) -> (then)
   > n8 = Branch(Condition=@n7)
   > Add(
    TargetArray:Array of WBP Tidal Crane Base Activities UI Reclamation Tracker Object References=Get SpawnersObjectivesEntries() -> (SpawnersObjectivesEntries:Array of WBP Tidal Crane Base Activities UI Reclamation Tracker Object References)
    NewItem:"WBP Tidal Crane Base Activities UI Reclamation Tracker Object Reference (by ref)"=@n6
   ) {is=title K2Node_CallArrayFunction}
   Add(
    execute=@n8.else
    TargetArray=Get NPCObjectivesEntries() -> (NPCObjectivesEntries:Array of WBP Tidal Crane Base Activities UI Reclamation Tracker Object References)
    NewItem=@n6
   )
  ]}
 EventGraph {nodes [
   OnEventComplete() {is=title K2Node_CustomEvent}
   > n1 = Branch(Condition=Get EventFinished() -> (EventFinished:bool))
   Set EventFinished(execute=@n1.else, EventFinished:bool=true) -> (Output_Get:bool)
   > "SetText (Text)"(
    self=Get Text_ReclamationMessage() -> (Text_ReclamationMessage:Common Text Object Reference)
    InText=Get TextCompletedEventMessage() -> (TextCompletedEventMessage:Text)
   )
   > PlayAnimation(InAnimation=Get Anim_Completed())
   > Delay(Duration=3)
   > PlayAnimation(InAnimation=Get Anim_Outro())
   OnEventFailed() {is=title K2Node_CustomEvent}
   > n2 = Branch(Condition="Equal (Enum)"(
     A:ESlateVisibility Enum=GetVisibility!(self:Widget Object Reference) -> (ReturnValue:ESlateVisibility Enum=Visible)
     B:ESlateVisibility Enum=Collapsed
    ))
   Set EventFinished(execute=@n2.else, EventFinished=true)
   > PlayAnimation(InAnimation=Get Anim_Outro())
   Event OnInitialized()
   > BindAnimationCallbacks() {is=title K2Node_CallFunction}
   > GetDynamicMaterials()
   Event Construct()
   > SetTouchLayout()
   > SetVisibility(InVisibility=Collapsed)
  ]}
 EventHasFinished {nodes [
   n1 = EventHasFinished!() -> (then, Spawn Manager:Spawn Manager Tidal Crane Reclaim Object Reference) {is=title K2Node_FunctionEntry}
   n2 = Get CurrentBase() -> (CurrentBase:Spawn Manager Tidal Crane Reclaim Object Reference)
   > Return Node!(execute, EventHasFinished:bool=AND Boolean(A="Equal (Object)"(A=@n1, B=@n2), B=Get EventFinished()))
  ]}
 GetDynamicMaterials {nodes [
   GetDynamicMaterials() {is=title K2Node_FunctionEntry}
   > n1 = GetDynamicMaterial(self=Get Image_ProgressBar() -> (Image_ProgressBar:Image Object Reference))
   > Set DynamicMaterial_ProgressBar(DynamicMaterial_ProgressBar:Material Instance Dynamic Object Reference=@n1) -> (Output_Get:Material Instance Dynamic Object Reference)
   > n2 = GetDynamicMaterial(self=Get Image_Glow() -> (Image_Glow:Image Object Reference))
   > Set DynamicMaterial_GlowBar(DynamicMaterial_GlowBar:Material Instance Dynamic Object Reference=@n2) -> (Output_Get:Material Instance Dynamic Object Reference)
  ]}
 HideObjectiveEntries {nodes [
   HideObjectiveEntries() {is=title K2Node_FunctionEntry}
   > n1 = For Each Loop(Array:Array of User Widget Object References=GetAllEntries(self=Get DynamicEntryBox_ObjectiveTracker())) -> (Array Element:User Widget Object Reference)
   > SetVisibility(self=@n1, InVisibility=Collapsed)
  ]}
 PlayIntroSFX {nodes [
   PlayIntroSFX() {is=title K2Node_FunctionEntry}
   > PlaySound2D(Sound=/TidalCrane_BaseActivities_UI/Assets/Sounds/MSS_UI_BaseReclamation_Popup)
  ]}
 PlayOutFeedback {nodes [
   PlayOutFeedback() {is=title K2Node_FunctionEntry}
   > PlayOutroSFX()
  ]}
 PlayOutroSFX {nodes [PlayOutroSFX() {is=title K2Node_FunctionEntry}]}
 Sequencer Events {nodes [
   PlayObjectiveEntriesAnimation() {is=title K2Node_CustomEvent}
   > n1 = For Each Loop with Delay!(Exec:Exec, Array:Array of User Widget Object References=GetAllEntries(self=Get DynamicEntryBox_ObjectiveTracker()), Delay:double=0.03) -> (Loop Body:Exec, Array Element:User Widget Object Reference, Array Index:int, Completed:Exec) {is=title K2Node_MacroInstance}
   > n2 = Cast To WBP_TidalCrane_BaseActivities_UI_ReclamationTracker(Object=@n1)
   > PlayIntroAnimation(self:WBP Tidal Crane Base Activities UI Reclamation Tracker Object Reference=@n2)
  ]}
 SetEventData {nodes [
   n1 = SetEventData!() -> (then, Base:Spawn Manager Tidal Crane Reclaim Object Reference, Event Data:Tidal Crane Base Reclamation Data Structure) {is=title K2Node_FunctionEntry}
   > Set ReclaMationData(ReclaMationData:Tidal Crane Base Reclamation Data Structure=@n1.Event Data) -> (Output_Get:Tidal Crane Base Reclamation Data Structure)
   n2 = GetAllEntries(self=Get DynamicEntryBox_ObjectiveTracker())
   n3 = Break Tidal Crane Base Reclamation Data!(TidalCrane_Base_Reclamation_Data:"Tidal Crane Base Reclamation Data Structure (by ref)"=@n1.Event Data) -> (
    SpawnManager_2_#13:Spawn Manager Tidal Crane Reclaim Object Reference
    EventState_5_#14:TidalCrane_Bases_Enum Enum
    FreedNPCs_8_#15:int
    MaxNPCsToFree_17_#16:int
    BugsKilled_10_#17:int
    SpawnersDestroyed_12_#18:int
    MaxSpawners_14_#19:int
   ) {is=title K2Node_BreakStruct}
   n4 = OR Boolean(
    A="Not Equal (Object)"(A:Object Reference=Get CurrentBase(), B:Object Reference=@n1) -> (ReturnValue:bool=~) {is=title K2Node_PromotableOperator}
    B="Not Equal (Integer)"(
     A=Length(TargetArray=@n2)
     B=int + int(A=@n3.MaxNPCsToFree_17_#16, B=@n3.MaxSpawners_14_#19)
    )
   )
   > Set NeedToSetUI(NeedToSetUI:bool=@n4) -> (Output_Get:bool)
   K2Node_Composite_0 = CompareCompletedObjectives!(A:int=@n3.FreedNPCs_8_#15, A2:int=@n3.SpawnersDestroyed_12_#18) -> (ReturnValue:bool=~) {is=title K2Node_Composite}
   > Branch(Condition=OR Boolean(A=@n4, B=@K2Node_Composite_0))
   > Set CurrentBase(CurrentBase:Spawn Manager Tidal Crane Reclaim Object Reference=@n1) -> (Output_Get:Spawn Manager Tidal Crane Reclaim Object Reference)
   > Set NeedObjectiveUpdate(NeedObjectiveUpdate:bool=@K2Node_Composite_0) -> (Output_Get:bool)
   > Set EventFinished(EventFinished=false)
   > "SetText (Text)"(self=Get Text_ReclamationMessage(), InText=Get TextDefaultEventMessage() -> (TextDefaultEventMessage:Text))
   > Set FreedNPcs(FreedNPcs:int=@n3.FreedNPCs_8_#15) -> (Output_Get:int=0)
   > Set MaxNpcs(MaxNpcs:int=@n3.MaxNPCsToFree_17_#16) -> (Output_Get:int=0)
   > Set SpawnersDestroyed(SpawnersDestroyed:int=@n3.SpawnersDestroyed_12_#18) -> (Output_Get:int=0)
   > Set MaxSpawners(MaxSpawners:int=@n3.MaxSpawners_14_#19) -> (Output_Get:int=0)
   n5 = Get FreedNPcs() -> (FreedNPcs:int=0)
   n6 = Get SpawnersDestroyed() -> (SpawnersDestroyed:int=0)
   > Set CurrentProgress(CurrentProgress:double="To Float (Integer)"(InInt=int + int(A=@n5, B=@n6))) -> (Output_Get:double)
   > Set CurrentMaxObjectivesCount(CurrentMaxObjectivesCount:int=int + int(A=Get MaxNpcs() -> (MaxNpcs:int=0), B=Get MaxSpawners() -> (MaxSpawners:int=0))) -> (Output_Get:int=0)
   > n7 = Sequence!(execute) -> (then_0, then_1, then_2:Exec, then_3:Exec)
   > CreateObjectiveTrackers!(execute, self, MaxObjectives:int=Get CurrentMaxObjectivesCount() -> (CurrentMaxObjectivesCount=0), MaxNpcs:int=Get MaxNpcs() -> (MaxNpcs=0)) -> (then) {is=title K2Node_CallFunction}
   SetObjectivesTrackers!(
    execute=@n7.then_1
    self
    ObjectiveArray:Array of WBP Tidal Crane Base Activities UI Reclamation Tracker Object References=Get NPCObjectivesEntries()
    ObjectivesCleared:int=Get FreedNPcs() -> (FreedNPcs=0)
   ) -> (then)
   SetObjectivesTrackers(execute=@n7.then_2, ObjectiveArray=Get SpawnersObjectivesEntries(), ObjectivesCleared=Get SpawnersDestroyed() -> (SpawnersDestroyed=0))
   n8 = Get!(execute:Exec=@n7.then_3) -> (then:Exec, else:Exec, NeedToSetUI:bool)
   n9 = float / float(A=Get CurrentProgress(), B=Get CurrentMaxObjectivesCount() -> (CurrentMaxObjectivesCount=0))
   n10 = Get ProgressOffset() -> (ProgressOffset:double)
   Set CurrentProgressBar(execute=[@n8.then, @n11.else], CurrentProgressBar:double=float - float(A=1, B=float + float(A=@n9, B=@n10))) -> (Output_Get:double)
   > CheckProgressBarOffset() {is=title K2Node_CallFunction}
   > SetProgressBar!(execute, self, Progress:double=Get CurrentProgressBar() -> (CurrentProgressBar:double)) -> (then)
   n11 = Get!(execute=@n8.else) -> (then, else, NeedObjectiveUpdate:bool)
   > UpdateEventHUD!(
    execute
    self
    Reclaim Data:Tidal Crane Base Reclamation Data Structure=Get ReclaMationData() -> (ReclaMationData:Tidal Crane Base Reclamation Data Structure)
   ) -> (then)
  ]}
 CompareCompletedObjectives {
  path=SetEventData.K2Node_Composite_0.CompareCompletedObjectives
  nodes [
   n1 = Inputs!() -> (A:int, A2:int)
   n2 = Get SpawnersDestroyed() -> (SpawnersDestroyed=0)
   n3 = Get FreedNPcs() -> (FreedNPcs=0)
   Outputs!(ReturnValue:bool=OR Boolean(A="integer > integer"(A=@n1, B=@n3), B="integer > integer"(A=@n1.A2, B=@n2)))
  ]
 }
 SetObjectivesState {nodes [
   n1 = SetObjectivesState!() -> (then, DynamicEntryBox:Dynamic Entry Box Object Reference, ObjectivesCleared:int) {is=title K2Node_FunctionEntry}
   > n2 = For Each Loop(Array=GetAllEntries(self=@n1))
   > n3 = Cast To WBP_TidalCrane_BaseActivities_UI_ReclamationTracker(Object=@n2)
   > SetObjectiveState!(
    execute
    self:WBP Tidal Crane Base Activities UI Reclamation Tracker Object Reference=@n3
    ObjectiveCleared:bool="integer >= integer"(A=@n2.Array Index, B=@n1.ObjectivesCleared)
   ) -> (then)
  ]}
 SetObjectivesTrackers {nodes [
   n1 = SetObjectivesTrackers!() -> (then, ObjectiveArray:Array of WBP Tidal Crane Base Activities UI Reclamation Tracker Object References, ObjectivesCleared:int) {is=title K2Node_FunctionEntry}
   > n2 = For Each Loop(Array:Array of WBP Tidal Crane Base Activities UI Reclamation Tracker Object References=@n1) -> (Array Element:WBP Tidal Crane Base Activities UI Reclamation Tracker Object Reference)
   > SetObjectiveState(self=@n2, ObjectiveCleared="integer < integer"(A=@n2.Array Index, B=@n1.ObjectivesCleared))
  ]}
 SetProgressBar {nodes [
   n1 = SetProgressBar!() -> (then, Progress:double=~) {is=title K2Node_FunctionEntry}
   > SetScalarParameterValue(
    self=Get DynamicMaterial_ProgressBar() -> (DynamicMaterial_ProgressBar:Material Instance Dynamic Object Reference)
    ParameterName=Progress
    Value=@n1
   )
   > SetScalarParameterValue(
    self=Get DynamicMaterial_GlowBar() -> (DynamicMaterial_GlowBar:Material Instance Dynamic Object Reference)
    ParameterName=Progress
    Value=@n1
   )
   > Set CurrentProgressBar(CurrentProgressBar=@n1)
  ]}
 SetTouchLayout {nodes [
   SetTouchLayout() {is=title K2Node_FunctionEntry}
   > Branch(Condition=IsUsingTouch(PlayerController=GetOwningPlayer()))
   > SetRenderScale!(execute, self:Widget Object Reference=Get DynamicEntryBox_ObjectiveTracker(), Scale:Vector 2D Structure=(X=1.2,Y=1.2)) -> (then)
   > SetHeightOverride(self=Get SizeBox_BaseReclamation() -> (SizeBox_BaseReclamation:Size Box Object Reference), InHeightOverride=55)
   > SetPadding!(
    execute
    self:Size Box Slot Object Reference=SlotAsSizeBoxSlot!(Widget:Widget Object Reference=Get VerticalBox_All() -> (VerticalBox_All:Vertical Box Object Reference)) -> (ReturnValue:Size Box Slot Object Reference)
    InPadding_Left:float=0
    InPadding_Top:float=-32
    InPadding_Right:float=0
    InPadding_Bottom:float=0
   ) -> (then)
   > n1 = Branch(Condition=IsTabletAspectRatio!() -> (ReturnValue:bool))
   > SetWidthOverride(self=Get SizeBox_BaseReclamation(), InWidthOverride=450)
   SetWidthOverride(execute=@n1.else, self=Get SizeBox_BaseReclamation(), InWidthOverride=600)
  ]}
 SetTrackerVisbility {nodes [
   SetTrackerVisbility() {is=title K2Node_FunctionEntry}
   > SetVisibility(InVisibility=Collapsed)
  ]}
 ShowReclamationUI {nodes [
   ShowReclamationUI() {is=title K2Node_FunctionEntry}
   > PlayAnimation(InAnimation=Get Anim_Intro() -> (Anim_Intro:Widget Animation Object Reference))
   > SetVisibility(InVisibility=HitTestInvisible)
   > PlayIntroSFX() {is=title K2Node_CallFunction}
  ]}
 UpdateEventHUD {nodes [
   PlayAnimationTimeRange!(
    execute
    self:User Widget Object Reference
    InAnimation:Widget Animation Object Reference=Get Anim_Progress() -> (Anim_Progress:Widget Animation Object Reference)
    StartAtTime:float=Get CurrentProgressBar()
    EndAtTime:float=float / float(A=Get CurrentProgress(), B=Get CurrentMaxObjectivesCount() -> (CurrentMaxObjectivesCount=0))
    NumLoopsToPlay:int=1
    PlayMode:EUMGSequencePlayMode Enum
    PlaybackSpeed:float=float * float(A=Get ProgressBarAnimationTime() -> (ProgressBarAnimationTime:double)) {is=title K2Node_PromotableOperator}
    bRestoreState:bool=false
   ) -> (then, ReturnValue:Widget Animation Handle Structure)
   > BindToAnimationFinished()
   n1 = UpdateEventHUD!() -> (then, Reclaim Data:Tidal Crane Base Reclamation Data Structure) {is=title K2Node_FunctionEntry}
   n2 = Break Tidal Crane Base Reclamation Data(TidalCrane_Base_Reclamation_Data=@n1)
   > Set FreedNPcs(FreedNPcs=@n2.FreedNPCs_8_#15) -> (Output_Get=0)
   > Set SpawnersDestroyed(SpawnersDestroyed=@n2.SpawnersDestroyed_12_#18) -> (Output_Get=0)
   n3 = Get SpawnersDestroyed() -> (SpawnersDestroyed=0)
   n4 = Get FreedNPcs() -> (FreedNPcs=0)
   > Set CurrentProgress(CurrentProgress="To Float (Integer)"(InInt=int + int(A=@n3, B=@n4)))
   > n5 = Sequence!(execute) -> (then_0, then_1, then_2)
   > UpdateObjectiveTrackers!(
    execute
    self
    ObjectiveArray:Array of WBP Tidal Crane Base Activities UI Reclamation Tracker Object References=Get NPCObjectivesEntries()
    ObjectivesCleared:int=Get FreedNPcs() -> (FreedNPcs=0)
   ) -> (then)
   UpdateObjectiveTrackers(execute=@n5.then_1, ObjectiveArray=Get SpawnersObjectivesEntries(), ObjectivesCleared=Get SpawnersDestroyed() -> (SpawnersDestroyed=0))
   CheckProgressBarOffset(execute=@n5.then_2)
   n6 = float / float(A=1, B:double=float + float(A:int=Get MaxSpawners() -> (MaxSpawners=0), B:int=Get MaxNpcs() -> (MaxNpcs=0)))
   > n7 = Set Timer by Event(
    Delegate=Create Event()
    Time=float / float(A=Get ProgressBarAnimationTime(), B=float / float(A=@n6, B=0.001))
    bLooping=true
    bMaxOncePerFrame=true
   )
   > Set TimerHandleStruct(TimerHandleStruct:Timer Handle Structure=@n7) -> (Output_Get:Timer Handle Structure)
  ]}
 UpdateObjectiveTrackers {nodes [
   n1 = UpdateObjectiveTrackers!() -> (then, ObjectiveArray:Array of WBP Tidal Crane Base Activities UI Reclamation Tracker Object References, ObjectivesCleared:int) {is=title K2Node_FunctionEntry}
   > n2 = For Each Loop(Array=@n1)
   > Branch(Condition="integer < integer"(A=@n2.Array Index, B=@n1.ObjectivesCleared))
   > UpdateObjectiveState(self:WBP Tidal Crane Base Activities UI Reclamation Tracker Object Reference=@n2)
  ]}
 UpdateProgressBarTimer {nodes [
   UpdateProgressBarTimer() {is=title K2Node_FunctionEntry}
   n1 = float + float(
    A:double=float / float(A=Get CurrentProgress(), B:int=Get CurrentMaxObjectivesCount() -> (CurrentMaxObjectivesCount=0))
    B:double=Get ProgressOffset()
   )
   > n2 = Branch(Condition="float <= float"(A=Get CurrentProgressBar(), B=float - float(A=1, B=@n1)) -> (ReturnValue=~))
   > Clear and Invalidate Timer by Handle(Handle=Get TimerHandleStruct() -> (TimerHandleStruct:Timer Handle Structure))
   n3 = Get CurrentProgress()
   n4 = Get CurrentMaxObjectivesCount() -> (CurrentMaxObjectivesCount=0)
   > n5 = Set CurrentProgressBar(CurrentProgressBar=float - float(A=1, B=float / float(A=@n3, B=@n4)))
   > Branch(Condition="Nearly Equal (Float)"(A=@n5, B=0, ErrorTolerance=0))
   > OnEventComplete() {is=title K2Node_CallFunction}
   SetScalarParameterValue(execute=@n2.else, self=Get DynamicMaterial_ProgressBar(), ParameterName=Progress, Value=Get CurrentProgressBar())
   > SetScalarParameterValue(self=Get DynamicMaterial_GlowBar(), ParameterName=Progress, Value=Get CurrentProgressBar())
   > Set CurrentProgressBar(CurrentProgressBar=float - float(A=Get CurrentProgressBar(), B=0.001))
  ]}
]
variables [
 CurrentBase {type=SpawnManager_TidalCrane_Reclaim_C, default=null}
 CurrentMaxObjectivesCount {type=integer}
 CurrentProgress {type=number}
 CurrentProgressBar {type=number, default=1}
 DynamicMaterial_GlowBar {type=MaterialInstanceDynamic}
 DynamicMaterial_ProgressBar {type=MaterialInstanceDynamic}
 EventFinished {type=boolean}
 FreedNPcs {type=integer}
 MaxNpcs {type=integer}
 MaxSpawners {type=integer}
 NeedObjectiveUpdate {type=boolean}
 NeedToSetUI {type=boolean}
 NPCObjectivesEntries {type="array<WBP_TidalCrane_BaseActivities_UI_ReclamationTracker_C>", default []}
 ObjectivesCleared {type=integer}
 ProgressBarAnimationTime {type=number, default=0.5}
 ProgressOffset {type=number}
 ReclaMationData {
  type=TidalCrane_Base_Reclamation_Data
  default {bugsKilled=0, eventState=Normal, freedNPCs=0, maxNPCsToFree=0, maxSpawners=0, spawnManager=null, spawnersDestroyed=0}
 }
 SpawnersDestroyed {type=integer}
 SpawnersObjectivesEntries {type="array<WBP_TidalCrane_BaseActivities_UI_ReclamationTracker_C>", default []}
 TextCompletedEventMessage {type=string, default="EVENT COMPLETED!"}
 TextDefaultEventMessage {type=string, default="Purge the outpost!"}
 TimerHandleStruct {type=TimerHandle}
]
widgets {
 info {inheritedWidgetCount=0, parentClass=&CommonUserWidget, rootWidgetClass=&SizeBox, widgetCount=12}
 widgets [
  SizeBox_BaseReclamation {
   widgetClassPath=&SizeBox
   bIsVariable=true
   widget=SizeBox_BaseReclamation
   properties {bOverride_WidthOverride=true, widthOverride=700}
  }
  VerticalBox_All {
   widgetClassPath=&VerticalBox
   bIsVariable=true
   parent=SizeBox_BaseReclamation
   widget=VerticalBox_All
   properties {bOverride_Cursor=true}
   slot {class=SizeBoxSlot, properties {verticalAlignment=VAlign_Top}}
  }
  Overlay_Progress {widgetClassPath=&Overlay, parent=VerticalBox_All, widget=Overlay_Progress, slot {class=VerticalBoxSlot}}
  Image_ProgressBar {
   widgetClassPath=&Image
   bIsVariable=true
   parent=Overlay_Progress
   widget=Image_ProgressBar
   properties {
    visibility=SelfHitTestInvisible
    brush {
     bIsDynamicallyLoaded=false
     drawAs=Image
     imageType=FullColor
     mirroring=NoMirror
     resourceName=None
     resourceObject=/TidalCrane_BaseActivities_UI/UI/Materials/MI_UI_Reclamation_Progress
     tiling=NoTile
     imageSize=<700 70>
     margin=<ltrb 0 0 0 0>
     outlineSettings {
      bUseBrushTransparency=false
      roundingType=HalfHeightRadius
      width=0
      color {colorUseRule=UseColor_Specified, specifiedColor=<rgba 0 0 0 0>}
      cornerRadii=<0 0 0 0>
     }
     tintColor {colorUseRule=UseColor_Specified, specifiedColor=<rgba 1 1 1 1>}
     uVRegion {bIsValid=false, max=<0 0>, min=<0 0>}
    }
   }
   slot {class=OverlaySlot}
  }
  Image_Glow {
   widgetClassPath=&Image
   bIsVariable=true
   parent=Overlay_Progress
   widget=Image_Glow
   properties {
    renderOpacity=0.1
    visibility=SelfHitTestInvisible
    brush {
     bIsDynamicallyLoaded=false
     drawAs=Image
     imageType=FullColor
     mirroring=NoMirror
     resourceName=None
     resourceObject=/TidalCraneSharedActivitiesUI/Materials/MI_UI_Reclamation_Progress_Glow
     tiling=NoTile
     imageSize=<32 32>
     margin=<ltrb 0 0 0 0>
     outlineSettings {
      bUseBrushTransparency=false
      roundingType=HalfHeightRadius
      width=0
      color {colorUseRule=UseColor_Specified, specifiedColor=<rgba 0 0 0 0>}
      cornerRadii=<0 0 0 0>
     }
     tintColor {colorUseRule=UseColor_Specified, specifiedColor=<rgba 1 1 1 1>}
     uVRegion {bIsValid=false, max=<0 0>, min=<0 0>}
    }
    renderTransform {angle=0, scale=<1 1>, shear=<0 0>, translation=<0 -32>}
   }
   slot {class=OverlaySlot, properties {horizontalAlignment=HAlign_Fill, verticalAlignment=VAlign_Fill}}
  }
  HorizontalBox_Objective {
   widgetClassPath=&HorizontalBox
   bIsVariable=true
   parent=VerticalBox_All
   widget=HorizontalBox_Objective
   properties {bOverride_Cursor=true}
   slot {class=VerticalBoxSlot, properties {verticalAlignment=VAlign_Center}}
  }
  Spacer_Left {
   widgetClassPath=&Spacer
   bIsVariable=true
   parent=HorizontalBox_Objective
   widget=Spacer_Left
   properties {size=<20 45>}
   slot {class=HorizontalBoxSlot}
  }
  DynamicEntryBox_ObjectiveTracker {
   widgetClassPath=&DynamicEntryBox
   bIsVariable=true
   parent=HorizontalBox_Objective
   widget=DynamicEntryBox_ObjectiveTracker
   properties {
    entryWidgetClass=&/WBP_TidalCrane_BaseActivities_UI_ReclamationTracker.WBP_TidalCrane_BaseActivities_UI_ReclamationTracker_C
    numDesignerPreviewEntries=4
    visibility=HitTestInvisible
   }
   slot {class=HorizontalBoxSlot, properties {horizontalAlignment=HAlign_Left}}
  }
  Spacer_Right {
   widgetClassPath=&Spacer
   bIsVariable=true
   parent=HorizontalBox_Objective
   widget=Spacer_Right
   properties {size=<28 45>}
   slot {class=HorizontalBoxSlot}
  }
  Overlay_48 {
   widgetClassPath=&Overlay
   parent=HorizontalBox_Objective
   widget=Overlay_48
   slot {class=HorizontalBoxSlot, properties {verticalAlignment=VAlign_Center}}
  }
  Text_ReclamationMessage {
   widgetClassPath=&CommonTextBlock
   bIsVariable=true
   parent=Overlay_48
   widget=Text_ReclamationMessage
   properties {
    mobileFontSizeMultiplier=1.2
    text="Purge the outpost!"
    textTransformPolicy=ToUpper
    wrapTextAt=500
    colorAndOpacity {colorUseRule=UseColor_Specified, specifiedColor=<rgba 1 1 1 0.9>}
    font {
     bForceMonospaced=false
     bMaterialIsStencil=false
     fontMaterial=null
     fontObject=&BurbankSmall
     letterSpacing=0
     monospacedWidth=1
     size=19.5
     skewAmount=0
     typefaceFontName=Black
     outlineSettings {
      bApplyOutlineToDropShadows=false
      bMiteredCorners=false
      bSeparateFillAlpha=false
      outlineMaterial=&M-FontShadow
      outlineSize=3
      outlineColor=<rgba 0 0 0 0.5>
     }
    }
    shadowOffset=<0 0>
    strikeBrush {
     bIsDynamicallyLoaded=false
     drawAs=Image
     imageType=FullColor
     mirroring=NoMirror
     resourceName=None
     resourceObject=null
     tiling=NoTile
     imageSize=<32 32>
     margin=<ltrb 0 0 0 0>
     outlineSettings {
      bUseBrushTransparency=false
      roundingType=HalfHeightRadius
      width=0
      color {colorUseRule=UseColor_Specified, specifiedColor=<rgba 0 0 0 0>}
      cornerRadii=<0 0 0 0>
     }
     tintColor {colorUseRule=UseColor_Specified, specifiedColor=<rgba 1 1 1 1>}
     uVRegion {bIsValid=false, max=<0 0>, min=<0 0>}
    }
   }
   slot {class=OverlaySlot}
  }
  Image_TextFX {
   widgetClassPath=&Image
   parent=Overlay_48
   widget=Image_TextFX
   properties {
    renderOpacity=0
    visibility=SelfHitTestInvisible
    brush {
     bIsDynamicallyLoaded=false
     drawAs=Image
     imageType=FullColor
     mirroring=NoMirror
     resourceName=None
     resourceObject=/TidalCraneSharedActivitiesUI/Materials/M_UI_TextBlink
     tiling=NoTile
     imageSize=<32 32>
     margin=<ltrb 0 0 0 0>
     outlineSettings {
      bUseBrushTransparency=false
      roundingType=HalfHeightRadius
      width=0
      color {colorUseRule=UseColor_Specified, specifiedColor=<rgba 0 0 0 0>}
      cornerRadii=<0 0 0 0>
     }
     tintColor {colorUseRule=UseColor_Specified, specifiedColor=<rgba 1 1 1 1>}
     uVRegion {bIsValid=false, max=<0 0>, min=<0 0>}
    }
    renderTransform {angle=0, scale=<2 2>, shear=<0 0>, translation=<0 0>}
   }
   slot {class=OverlaySlot, properties {horizontalAlignment=HAlign_Fill, verticalAlignment=VAlign_Fill}}
  }
 ]
}
