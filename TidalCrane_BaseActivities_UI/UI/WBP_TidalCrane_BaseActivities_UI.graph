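# HUD widget for TidalCrane base activities that switches between Defense and Reclamation panels for the base the player is in.
# Construct retries until the playspace Gameplay Event Router exists, then listens for defense, timer and reclamation events, filtered to ActiveBase by a macro.
# An overlap stateful event, re-subscribed on each view target change, sets or clears the active base.
# The base's type and event state decide which panel shows, and leaving collapses the widget.
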WidgetBlueprint /TidalCrane_BaseActivities_UI/UI/WBP_TidalCrane_BaseActivities_UI
graphs [
 CheckActiveBase {nodes [
   n1 = Inputs!() -> (execute, CurrentBase:Object Reference)
   > Branch(Condition="Equal (Object)"(A=Get ActiveBase() -> (ActiveBase:Actor Object Reference), B=@n1))
   > Outputs!(then:Exec)
  ]}
 EventGraph {nodes [
   n1 = OnViewTargetChanged!() -> (
    OutputDelegate
    then
    PlayerController:Fort Player Controller Object Reference
    OldViewTarget:Actor Object Reference
    NewViewTarget:Actor Object Reference
   ) {is=title K2Node_CustomEvent}
   > StopListeningToStatefulEvent(self:Async Action Start Listening To Stateful Event Object Reference=Get AsyncTaskRef() -> (AsyncTaskRef:Async Action Start Listening To Stateful Event Object Reference))
   > n2 = StartListeningToStatefulEvent!(
    execute:Exec
    Target:Gameplay Event Router Component Object Reference=Get GameplayEventRouterComponent() -> (GameplayEventRouterComponent:Gameplay Event Router Component Object Reference)
    EventType:Script Struct Object Reference=/TidalCrane_BaseActivities/Gameplay/TidalCrane_Defense_OverlapStatefulEvent
    Context:Object Reference=@n1.NewViewTarget
    EventBubblingRule:EEventBubblingRule Enum
   ) -> (
    then:Exec
    AsyncTaskProxy:Async Action Start Listening To Stateful Event Object Reference
    OnEventReceived:Exec
    OnSavedEventStateExists:Exec
    OnEventStateCleared:Exec
    Payload:Tidal Crane Defense Overlap Stateful Event Structure
   ) {is=title K2Node_StartListeningToStatefulEventAsyncAction}
   > Set AsyncTaskRef(AsyncTaskRef:Async Action Start Listening To Stateful Event Object Reference=@n2) -> (Output_Get:Async Action Start Listening To Stateful Event Object Reference)
   > Is Valid(InputObject=Get ActiveBase())
   > n3 = Cast To FortPlayerPawn(Object=@n1.NewViewTarget)
   > n4 = Cast To B_TidalCrane_MilitaryActivity_SpawnManager!(execute, Object=Get ActiveBase()) -> (then, CastFailed, AsB Tidal Crane Military Activity Spawn Manager:B Tidal Crane Military Activity Spawn Manager Object Reference) {is=title K2Node_DynamicCast}
   > n5 = CheckIfPlayersIsOverlapping!(execute, self:B Tidal Crane Military Activity Spawn Manager Object Reference=@n4, Spectator:Fort Player Pawn Object Reference=@n3) -> (then, IsOverlapping:bool)
   > n6 = Branch(Condition=@n5)
   n7 = Break Tidal Crane Defense Overlap Stateful Event!(TidalCrane_Defense_OverlapStatefulEvent:"Tidal Crane Defense Overlap Stateful Event Structure (by ref)"=Get TidalCraneOverlapStatefulEvent() -> (TidalCraneOverlapStatefulEvent:Tidal Crane Defense Overlap Stateful Event Structure)) -> (
    SpawnManager_18_#1:Actor Soft Object Reference
    BaseType_17_#2:TidalCrane_BaseTypes Enum
    Player_19_#3:Fort Pawn Soft Object Reference
    bIsOverlapping_10_#4:bool=~
   ) {is=title K2Node_BreakStruct}
   UpdateActiveBase!(
    execute=@n6.else
    self
    OverlapEvent:Tidal Crane Defense Overlap Stateful Event Structure=Make Tidal Crane Defense Overlap Stateful Event!(
     SpawnManager_18_#1:Actor Soft Object Reference=@n7
     BaseType_17_#2:TidalCrane_BaseTypes Enum=@n7.BaseType_17_#2
     Player_19_#3:Fort Pawn Soft Object Reference=@n7.Player_19_#3
     bIsOverlapping_10_#4:bool=@n5
    ) -> (TidalCrane_Defense_OverlapStatefulEvent:Tidal Crane Defense Overlap Stateful Event Structure) {is=title K2Node_MakeStruct}
   ) -> (then)
   UpdateActiveBase(execute=[@n2.OnEventReceived, @n2.OnSavedEventStateExists], OverlapEvent=@n2.Payload)
   Event On Become Active() {is=title K2Node_Event}
   > n8 = Switch on TidalCrane_BaseTypes!(execute, Selection:TidalCrane_BaseTypes Enum=Get ActiveBaseType() -> (ActiveBaseType:TidalCrane_BaseTypes Enum=NewEnumerator4)) -> (NewEnumerator4:Exec, NewEnumerator0:Exec, NewEnumerator1:Exec, NewEnumerator2:Exec, NewEnumerator3:Exec) {is=title K2Node_SwitchEnum}
   n9 = Cast To SpawnManager_TidalCrane_Defense!(execute=@n8.NewEnumerator1, Object=Get ActiveBase()) -> (then, CastFailed, AsSpawn Manager Tidal Crane Defense:Spawn Manager Tidal Crane Defense Object Reference) {is=title K2Node_DynamicCast}
   > n10 = GetEventData!(execute, self:Spawn Manager Tidal Crane Defense Object Reference=@n9) -> (then, NewParam:Tidal Crane Base Defense Data Structure)
   > n11 = GetProgressEventData!(
    execute
    self:Spawn Manager Tidal Crane Defense Object Reference=Break Tidal Crane Base Defense Data!(TidalCrane_Base_DefenseData:"Tidal Crane Base Defense Data Structure (by ref)"=@n10) -> (
     SpawnManager_14_#5:Spawn Manager Tidal Crane Defense Object Reference
     EventState_10_#6:TidalCrane_Bases_Enum Enum
     MaxBugs_27_#7:int
    ) {is=title K2Node_BreakStruct}
   ) -> (then, EventData:Tidal Crane Base Timer Event Structure)
   > ShowEventHUD!(
    execute
    self:WBP Tidal Crane Base Activities UI Defense Object Reference=Get WBP_TidalCrane_BaseActivities_UI_Defense() -> (WBP_TidalCrane_BaseActivities_UI_Defense:WBP Tidal Crane Base Activities UI Defense Object Reference)
    Payload:Tidal Crane Base Timer Event Structure=@n11
   ) -> (then)
   ShowReclamationUI(
    execute=@n8.NewEnumerator3
    self:WBP Tidal Crane Base Activities UI Reclamation UI Object Reference=Get WBP_TidalCrane_BaseActivities_UI_ReclamationUI() -> (WBP_TidalCrane_BaseActivities_UI_ReclamationUI:WBP Tidal Crane Base Activities UI Reclamation UI Object Reference)
   )
   Event Destruct()
   > n12 = Cast To FortPlayerController(Object=GetOwningPlayer())
   > Unbind Event from On View Target Changed(self:Fort Player Controller Object Reference=@n12, Delegate=Create Event()) {is=title K2Node_RemoveDelegate}
   > StopListeningToStatefulEvent(self=Get AsyncTaskRef())
   Event Construct()
   > n13 = Set IsConstructed(IsConstructed:bool=false) -> (Output_Get:bool)
   n14 = GetEventRouter!(
    execute=[@n13.then, @n16.else]
    Scope:Object Class Reference=/Script/FortniteGame.FortPlayspace
    ContextActor:Actor Object Reference=GetOwningPlayer()
   ) -> (then, ReturnValue:Gameplay Event Router Component Object Reference)
   > n15 = Is Valid(InputObject=@n14)
   Delay(execute=@n15.Is Not Valid, Duration=5)
   > n16 = Get!(execute:Exec) -> (then:Exec, else:Exec, IsConstructed:bool)
   Set IsConstructed(execute=@n15.Is Valid, IsConstructed=true)
   > Set GameplayEventRouterComponent(GameplayEventRouterComponent:Gameplay Event Router Component Object Reference=@n14) -> (Output_Get:Gameplay Event Router Component Object Reference)
   > n17 = Sequence!(execute) -> (then_0, then_1, then_2:Exec, then_3:Exec, then_4:Exec)
   > n18 = Cast To FortPlayerController(Object=GetOwningPlayer())
   > Bind Event to On View Target Changed(self:Fort Player Controller Object Reference=@n18, Delegate=Create Event()) {is=title K2Node_AddDelegate}
   > OnViewTargetChanged!(
    execute
    self
    PlayerController:Fort Player Controller Object Reference
    OldViewTarget:Actor Object Reference
    NewViewTarget:Actor Object Reference=GetViewTarget!(self:Controller Object Reference=@n18) -> (ReturnValue:Actor Object Reference)
   ) -> (then) {is=title K2Node_CallFunction}
   n19 = Get GameplayEventRouterComponent()
   n20 = StartListeningToEvent!(
    execute:Exec=@n17.then_1
    Target:Gameplay Event Router Component Object Reference=@n19
    EventType:Script Struct Object Reference=/TidalCrane_BaseActivities/Gameplay/TidalCrane_Base_DefenseData
    Context:Object Reference
    EventBubblingRule:EEventBubblingRule Enum
    None:%=@n19
   ) -> (
    then:Exec
    AsyncTaskProxy:Async Action Start Listening To Event Object Reference
    OnEventReceived:Exec
    Payload:Tidal Crane Base Defense Data Structure
   )
   n21 = Break Tidal Crane Base Defense Data(TidalCrane_Base_DefenseData=@n20.Payload)
   Check Active Base!(execute:Exec=@n20.OnEventReceived, CurrentBase:Object Reference=@n21) -> (then:Exec) {is=title K2Node_MacroInstance}
   > UpdateDefenseData!(
    execute
    self:WBP Tidal Crane Base Activities UI Defense Object Reference=Get WBP_TidalCrane_BaseActivities_UI_Defense()
    Defense Data:Tidal Crane Base Defense Data Structure=@n20.Payload
   ) -> (then)
   > n22 = Branch(Condition="Equal (Enum)"(A:TidalCrane_Bases_Enum Enum=@n21.EventState_10_#6, B:TidalCrane_Bases_Enum Enum=NewEnumerator2))
   > SetActiveWidgetIndex(
    self:Common Visibility Switcher Object Reference=Get CommonVisibilitySwitcher_Container() -> (CommonVisibilitySwitcher_Container:Common Visibility Switcher Object Reference)
    Index=0
   )
   > SetVisibility(InVisibility=HitTestInvisible)
   Branch(execute=@n22.else, Condition="Equal (Enum)"(A=@n21.EventState_10_#6, B=NewEnumerator6))
   > OnEventFailed(self:WBP Tidal Crane Base Activities UI Defense Object Reference=Get WBP_TidalCrane_BaseActivities_UI_Defense())
   n23 = StartListeningToEvent(execute=@n17.then_2, Target=Get GameplayEventRouterComponent(), EventType=/TidalCrane_BaseActivities/Gameplay/TidalCrane_Base_TimerEvent) -> (Payload:Tidal Crane Base Timer Event Structure)
   Check Active Base(
    execute=@n23.OnEventReceived
    CurrentBase=Break Tidal Crane Base Timer Event!(TidalCrane_Base_TimerEvent:"Tidal Crane Base Timer Event Structure (by ref)"=@n23.Payload) -> (SpawnManagerActor_5_#8:Spawn Manager Tidal Crane Defense Object Reference) {is=title K2Node_BreakStruct}
   )
   > ShowEventHUD(self=Get WBP_TidalCrane_BaseActivities_UI_Defense(), Payload=@n23.Payload)
   n24 = StartListeningToEvent(
    execute=@n17.then_3
    Target=Get GameplayEventRouterComponent()
    EventType=/TidalCrane_BaseActivities/Gameplay/TidalCrane_Base_Reclamation_Data
   ) -> (Payload:Tidal Crane Base Reclamation Data Structure)
   n25 = Break Tidal Crane Base Reclamation Data!(TidalCrane_Base_Reclamation_Data:"Tidal Crane Base Reclamation Data Structure (by ref)"=@n24.Payload) -> (
    SpawnManager_2_#9:Spawn Manager Tidal Crane Reclaim Object Reference
    EventState_5_#10:TidalCrane_Bases_Enum Enum
    FreedNPCs_8_#11:int
    MaxNPCsToFree_17_#12:int
    BugsKilled_10_#13:int
    SpawnersDestroyed_12_#14:int
    MaxSpawners_14_#15:int
   ) {is=title K2Node_BreakStruct}
   Check Active Base(execute=@n24.OnEventReceived, CurrentBase=@n25)
   > SetEventData!(
    execute
    self:WBP Tidal Crane Base Activities UI Reclamation UI Object Reference=Get WBP_TidalCrane_BaseActivities_UI_ReclamationUI()
    Base:Spawn Manager Tidal Crane Reclaim Object Reference=@n25
    Event Data:Tidal Crane Base Reclamation Data Structure=@n24.Payload
   ) -> (then)
   > n26 = Branch(Condition=OR Boolean(
     A="Equal (Enum)"(A=@n25.EventState_5_#10, B=NewEnumerator0)
     B="Equal (Enum)"(A=@n25.EventState_5_#10, B=NewEnumerator2)
    ))
   > SetActiveWidgetIndex(self=Get CommonVisibilitySwitcher_Container(), Index=1)
   > SetVisibility(InVisibility=HitTestInvisible)
   Branch(execute=@n26.else, Condition="Equal (Enum)"(A=@n25.EventState_5_#10, B=NewEnumerator6))
   > OnEventFailed(self:WBP Tidal Crane Base Activities UI Reclamation UI Object Reference=Get WBP_TidalCrane_BaseActivities_UI_ReclamationUI())
  ]}
 UpdateActiveBase {nodes [
   n1 = UpdateActiveBase!() -> (then, OverlapEvent:Tidal Crane Defense Overlap Stateful Event Structure) {is=title K2Node_FunctionEntry}
   > Set TidalCraneOverlapStatefulEvent(TidalCraneOverlapStatefulEvent:Tidal Crane Defense Overlap Stateful Event Structure=@n1) -> (Output_Get:Tidal Crane Defense Overlap Stateful Event Structure)
   n2 = Break Tidal Crane Defense Overlap Stateful Event(TidalCrane_Defense_OverlapStatefulEvent=@n1) -> (bIsOverlapping_10_#4=~)
   > Is Valid(InputObject=Resolve Soft Reference!(Input:Fort Pawn Soft Object Reference=@n2.Player_19_#3) -> (Output:Fort Pawn Object Reference))
   > n3 = Branch(Condition=@n2.bIsOverlapping_10_#4)
   > Set ActiveBase(ActiveBase:Actor Object Reference=Resolve Soft Reference(Input:Actor Soft Object Reference=@n2) -> (Output:Actor Object Reference)) -> (Output_Get:Actor Object Reference)
   > n4 = Set ActiveBaseType(ActiveBaseType:TidalCrane_BaseTypes Enum=@n2.BaseType_17_#2) -> (Output_Get:TidalCrane_BaseTypes Enum=NewEnumerator4)
   > n5 = Switch on TidalCrane_BaseTypes(Selection=@n4)
   n6 = Cast To SpawnManager_TidalCrane_Defense(execute=@n5.NewEnumerator1, Object=Get ActiveBase())
   > n7 = GetEventData(self=@n6)
   n8 = Break Tidal Crane Base Defense Data(TidalCrane_Base_DefenseData=@n7)
   > Branch(Condition="Equal (Enum)"(A=@n8.EventState_10_#6, B=NewEnumerator2))
   > Is Valid(InputObject=@n8)
   > SetDefenseData!(
    execute
    self:WBP Tidal Crane Base Activities UI Defense Object Reference=Get WBP_TidalCrane_BaseActivities_UI_Defense()
    Base:Spawn Manager Tidal Crane Defense Object Reference=@n6
    Defense Data:Tidal Crane Base Defense Data Structure=@n7
   ) -> (then)
   > SetActiveWidgetIndex(self=Get CommonVisibilitySwitcher_Container(), Index=0)
   > SetVisibility(InVisibility=HitTestInvisible)
   n9 = Cast To SpawnManager_TidalCrane_Reclaim!(execute=@n5.NewEnumerator3, Object=Get ActiveBase()) -> (then, CastFailed, AsSpawn Manager Tidal Crane Reclaim:Spawn Manager Tidal Crane Reclaim Object Reference) {is=title K2Node_DynamicCast}
   n10 = Get EventData!(self:Spawn Manager Tidal Crane Reclaim Object Reference=@n9) -> (EventData:Tidal Crane Base Reclamation Data Structure)
   n11 = Break Tidal Crane Base Reclamation Data(TidalCrane_Base_Reclamation_Data=@n10)
   > Branch(Condition=OR Boolean(
     A="Equal (Enum)"(A=@n11.EventState_5_#10, B=NewEnumerator0)
     B="Equal (Enum)"(A=@n11.EventState_5_#10, B=NewEnumerator2)
    ))
   > Is Valid(InputObject=@n11)
   > SetEventData(self=Get WBP_TidalCrane_BaseActivities_UI_ReclamationUI(), Base=@n9, Event Data=@n10)
   > SetActiveWidgetIndex(self=Get CommonVisibilitySwitcher_Container(), Index=1)
   > SetVisibility(InVisibility=HitTestInvisible)
   n12 = Branch(
    execute=@n3.else
    Condition="Equal (Enum)"(A:TidalCrane_BaseTypes Enum=Get ActiveBaseType() -> (ActiveBaseType=NewEnumerator4), B:TidalCrane_BaseTypes Enum=NewEnumerator3)
   )
   > n13 = HideObjectiveEntries(self:WBP Tidal Crane Base Activities UI Reclamation UI Object Reference=Get WBP_TidalCrane_BaseActivities_UI_ReclamationUI())
   Set ActiveBaseType(execute=[@n13.then, @n12.else], ActiveBaseType=NewEnumerator4) -> (Output_Get=NewEnumerator4)
   > Set ActiveBase()
   > SetVisibility(InVisibility=Collapsed)
  ]}
]
variables [
 ActiveBase {type=Actor, default=null}
 ActiveBaseType {type=TidalCrane_BaseTypes, default=None}
 AsyncTaskRef {type=AsyncAction_StartListeningToStatefulEvent, default=null}
 CurrentNPCs {type="array<FortPlayerPawn>", default []}
 EventMaxDurationTime {type=number, default=120}
 EventMessage {type=string, default="Protect the base!"}
 EventTime {type=number}
 GameplayEventRouterComponent {type=GameplayEventRouterComponent, default=null}
 IsConstructed {type=boolean}
 NPCsAlive {type=integer}
 TidalCraneOverlapStatefulEvent {type=TidalCrane_Defense_OverlapStatefulEvent, default {bIsOverlapping=false, baseType=None, player=null, spawnManager=null}}
 TimerHandleStructure {type=TimerHandle}
]
widgets {
 info {
  inheritedWidgetCount=1
  parentClass=/Script/FortniteUI.AthenaCommonPrioritizedWidget
  rootWidgetClass=&CommonVisibilitySwitcher
  widgetCount=4
 }
 widgets [
  CommonVisibilitySwitcher_Container {
   widgetClassPath=&CommonVisibilitySwitcher
   bIsVariable=true
   widget=CommonVisibilitySwitcher_Container
   properties {activeWidgetIndex=1}
  }
  WBP_TidalCrane_BaseActivities_UI_Defense {
   widgetClassPath=&/WBP_TidalCrane_BaseActivities_UI_Defense
   bIsVariable=true
   parent=CommonVisibilitySwitcher_Container
   widget=WBP_TidalCrane_BaseActivities_UI_Defense
   slot {class=CommonVisibilitySwitcherSlot}
  }
  WBP_TidalCrane_BaseActivities_UI_ReclamationUI {
   widgetClassPath=&/WBP_TidalCrane_BaseActivities_UI_ReclamationUI
   bIsVariable=true
   parent=CommonVisibilitySwitcher_Container
   widget=WBP_TidalCrane_BaseActivities_UI_ReclamationUI
   slot {class=CommonVisibilitySwitcherSlot, properties {horizontalAlignment=HAlign_Left, verticalAlignment=VAlign_Top}}
  }
  TutorialHighlight {widgetClassPath=&TutorialHighlight}
 ]
}
